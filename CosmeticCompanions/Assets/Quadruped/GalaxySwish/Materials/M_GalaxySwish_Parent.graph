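# Parent material for the GalaxySwish quadruped cosmetic companion, adding an animated galaxy effect to the shared character shading with custom companion color.
# Two triplanar virtual-box nebula layers are hue-shifted, tinted and screen-blended, producing a mask for two twinkling star layers.
# Time-of-day color tints the galaxy.
# Channel masks, optional surface overrides, Fresnel in the primary color and an accessory glow color control the final blend between character and galaxy looks.
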Material /CosmeticCompanions/Assets/Quadruped/GalaxySwish/Materials/M_GalaxySwish_Parent
graph {
 properties {
  bAutomaticallySetUsageInEditor=false
  bUseMaterialAttributes=true
  bUsedWithMeshDeformer=true
  bUsedWithMorphTargets=true
  bUsedWithSkeletalMesh=true
  bUsedWithStaticLighting=true
  shadingModel=MSM_FromMaterialExpression
  usedShadingModels=Default Lit | Subsurface | Clear Coat | Subsurface Profile | Cloth
 }
 nodes [
  n1 = Add(A=NamedRerouteUsage(), B=NamedRerouteUsage())
  n2 = VectorParameter() {group=Galaxy, parameterName=Accessory Glow Color, sortPriority=91, defaultValue=<rgba 0.3281 0.0401 0.2694 1>}
  n3 = ChannelMaskParameter(None=NamedRerouteUsage()) {group=Galaxy, maskChannel=Blue, parameterName=AccessoryGlowChannel, sortPriority=-10, defaultValue=<rgba 0 0 1 0>}
  LinearInterpolate(A=@n1, B=@n2.RGB, Alpha=@n3.R)
  n4 = NamedRerouteUsage()
  LinearInterpolate(A=@n4, B=@n2.RGB, Alpha=@n3.R)
  n5 = MaterialFunctionCall() {materialFunction=&MF_CharacterParent_2023_PreFX}
  MaterialFunctionCall(Material=NamedRerouteUsage(), M=@n5.M, Pre Skinned Local Position=@n5.M, Distance Blend=@n5.M, Fresnel=@n5.M) {materialFunction=&MF_CharacterParent_2023_PostFX}
  NamedRerouteDeclaration(None=@n5.Tech Art Mask) {name=TechArtMask, nodeColor=<rgba 0 1 0.6953 1>}
  n6 = NamedRerouteUsage()
  n7 = NamedRerouteUsage()
  n8 = NamedRerouteUsage()
  n9 = ScalarParameter() {defaultValue=0.01, group=Galaxy - Stars, parameterName=Galaxy.Stars_TwinkleFrequency_Layer1, sortPriority=15}
  n10 = ScalarParameter() {defaultValue=0.7, group=Galaxy - Stars, parameterName=Galaxy.Stars_TwinkleSpeed_Layer1, sortPriority=16}
  n11 = Max(
   A=Multiply(
    A=MaterialFunctionCall!(Mask=@n6, UVs=@n7, RandomOffset=@n8, BlinkFrequency=@n9, BlinkSpeed=@n10) {materialFunction=/Game/Global/Materials/Character_Parent_2023/RandomBlink_v1/MF_RandomBlink}.Result
    B=ScalarParameter() {defaultValue=5, group=Galaxy - Stars, parameterName=Galaxy.Stars_TwinkleBrightness_Layer1, sortPriority=17}
   )
   B=NamedRerouteUsage()
  )
  n12 = ScalarParameter() {defaultValue=1, group=Galaxy - Stars, parameterName=Galaxy.Stars_Brightness_Layer1, sortPriority=14}
  n13 = NamedRerouteUsage()
  n14 = NamedRerouteUsage()
  n15 = NamedRerouteUsage()
  n16 = ScalarParameter() {defaultValue=0.02, group=Galaxy - Stars, parameterName=Galaxy.Stars_TwinkleFrequency_Layer2, sortPriority=25}
  n17 = ScalarParameter() {defaultValue=0.5, group=Galaxy - Stars, parameterName=Galaxy.Stars_TwinkleSpeed_Layer2, sortPriority=26}
  n18 = Max(
   A=Multiply(
    A=MaterialFunctionCall(Mask=@n13, UVs=@n14, RandomOffset=@n15, BlinkFrequency=@n16, BlinkSpeed=@n17) {materialFunction=&MF_RandomBlink}.Result
    B=ScalarParameter() {defaultValue=5, group=Galaxy - Stars, parameterName=Galaxy.Stars_TwinkleBrightness_Layer2, sortPriority=27}
   )
   B=NamedRerouteUsage()
  )
  n19 = ScalarParameter() {defaultValue=1, group=Galaxy - Stars, parameterName=Galaxy.Stars_Brightness_Layer2, sortPriority=24}
  NamedRerouteDeclaration(None=Max(A=Multiply(A=@n11, B=@n12), B=Multiply(A=@n18, B=@n19))) {name=Stars, nodeColor=<rgba 0 1 0.5078 1>}
  n20 = ScalarParameter() {group=Galaxy - Nebula, parameterName=Galaxy.Nebula_HueShift_Layer1, sortPriority=17}
  n21 = TextureObjectParameter() {
   group=Galaxy - Nebula
   parameterName=Galaxy.NebulaTexture_Layer1
   sortPriority=10
   texture=/Game/Global/Materials/Character_Parent_2023/Galaxy_Parent_2025/Textures/T_FN_Nebula_style_blue
  }
  n22 = VectorParameter() {group=Galaxy - Nebula, parameterName=Galaxy.Nebula_Speed_Layer1, sortPriority=14, defaultValue=<rgba 0.002 -0.002 0 1>}
  n23 = VectorParameter() {group=Galaxy - General Parameters, parameterName=Galaxy.OriginOffset, sortPriority=10, defaultValue=<rgba 0 0 0 1>}
  n24 = MaterialFunctionCall!(
   Texture=@n21
   Depth=ScalarParameter() {defaultValue=100, group=Galaxy - Nebula, parameterName=Galaxy.Nebula_Depth_Layer1, sortPriority=11}
   Tile=ScalarParameter() {defaultValue=0.4, group=Galaxy - Nebula, parameterName=Galaxy.Nebula_Tile_Layer1, sortPriority=12}
   BlendContrast=ScalarParameter() {defaultValue=0.4, group=Galaxy - Nebula, parameterName=Galaxy.Nebula_BlendContrast_Layer1, sortPriority=13}
   SpeedXY=ComponentMask(None=@n22.RGB) {g=true, r=true}
   ActorOffset=NamedRerouteDeclaration(None=@n23.RGB) {name=ActorOffset, nodeColor=<rgba 0.9688 1 0 1>}
   bUseLocalCameraVector
   bUseDitherSingleSample
  ) {materialFunction=/Game/Global/MaterialFunctions/MF_TriplanarVirtualBox}
  n25 = Desaturation(
   None=Power(
    Base=MaterialFunctionCall(Hue Shift Percentage=@n20, Texture=@n24.Result) {materialFunction=&HueShift}.Result
    Exp=ScalarParameter() {defaultValue=1.5, group=Galaxy - Nebula, parameterName=Galaxy.Nebula_Power_Layer1, sortPriority=16}
   )
   Fraction=ScalarParameter() {defaultValue=0.35, group=Galaxy - Nebula, parameterName=Galaxy.Nebula_Desaturation_Layer1, sortPriority=18}
  )
  n26 = VectorParameter() {parameterName=Galaxy_01_Tint, defaultValue=<rgba 1 1 1 0>}
  NamedRerouteDeclaration(None=MaterialFunctionCall(
    LDR Input=Multiply(A=@n25, B=@n26.RGB)
    HDR Intensity Multiplier=ScalarParameter() {defaultValue=0.6, group=Galaxy - Nebula, parameterName=Galaxy.Nebula_Brightness_Layer1, sortPriority=15}
   ) {materialFunction=&DeriveHDRfromLDR}.RemappedRange) {name=Galaxy_01, nodeColor=<rgba 1 0.2578 0 1>}
  NamedRerouteDeclaration(None=Add(A=NamedRerouteUsage(), B=NamedRerouteUsage())) {name=OUT_Emissive, nodeColor=<rgba 1 0 0.1875 1>}
  n27 = SetMaterialAttributes!(
   MaterialAttributes=NamedRerouteUsage()
   Base Color=@n4
   Metallic=NamedRerouteUsage()
   Specular=NamedRerouteUsage()
   Roughness=NamedRerouteUsage()
   Emissive Color=@n1
  ) {attributeSetTypes [
    #1
    #2
    #3
    #4
    #5
   ]}
  NamedRerouteDeclaration(None=@n27) {name=OUT_Material, nodeColor=<rgba 0 1 0.625 1>}
  n28 = ScalarParameter() {defaultValue=0.05, group=Galaxy - Nebula, parameterName=Galaxy.Nebula_HueShift_Layer2, sortPriority=27}
  n29 = TextureObjectParameter() {
   group=Galaxy - Nebula
   parameterName=Galaxy.NebulaTexture_Layer2
   sortPriority=20
   texture=/Game/Global/Materials/Character_Parent_2023/Galaxy_Parent_2025/Textures/T_FN_Nebula_style2
  }
  n30 = VectorParameter() {group=Galaxy - Nebula, parameterName=Galaxy.Nebula_Speed_Layer2, sortPriority=24, defaultValue=<rgba -0.001 0.0005 0 1>}
  n31 = MaterialFunctionCall(
   Texture=@n29
   Depth=ScalarParameter() {defaultValue=50, group=Galaxy - Nebula, parameterName=Galaxy.Nebula_Depth_Layer2, sortPriority=21}
   Tile=ScalarParameter() {defaultValue=0.4, group=Galaxy - Nebula, parameterName=Galaxy.Nebula_Tile_Layer2, sortPriority=22}
   BlendContrast=ScalarParameter() {defaultValue=1, group=Galaxy - Nebula, parameterName=Galaxy.Nebula_BlendContrast_Layer2, sortPriority=23}
   SpeedXY=ComponentMask(None=@n30.RGB) {g=true, r=true}
   ActorOffset=NamedRerouteUsage()
  ) {materialFunction=&MF_TriplanarVirtualBox}
  n32 = Desaturation(
   None=Power(
    Base=MaterialFunctionCall(Hue Shift Percentage=@n28, Texture=@n31.Result) {materialFunction=&HueShift}.Result
    Exp=ScalarParameter() {defaultValue=0.85, group=Galaxy - Nebula, parameterName=Galaxy.Nebula_Power_Layer2, sortPriority=26}
   )
   Fraction=ScalarParameter() {group=Galaxy - Nebula, parameterName=Galaxy.Nebula_Desaturation_Layer2, sortPriority=28}
  )
  n33 = VectorParameter() {parameterName=Galaxy_02_Tint, defaultValue=<rgba 1 1 1 0>}
  NamedRerouteDeclaration(None=MaterialFunctionCall(
    LDR Input=Multiply(A=@n32, B=@n33.RGB)
    HDR Intensity Multiplier=ScalarParameter() {defaultValue=0.5, group=Galaxy - Nebula, parameterName=Galaxy.Nebula_Brightness_Layer2, sortPriority=25}
   ) {materialFunction=&DeriveHDRfromLDR}.RemappedRange) {name=Galaxy_02, nodeColor=<rgba 0.0781 0 1 1>}
  n34 = Constant() {r=1}
  n35 = Multiply(A=Subtract(A=Constant() {r=1}, B=NamedRerouteUsage()), B=Subtract(A=Constant() {r=1}, B=NamedRerouteUsage()))
  n36 = Desaturation(None=NamedRerouteDeclaration(None=Subtract(A=@n34, B=@n35)) {name=Galaxies, nodeColor=<rgba 1 0 0.2578 1>}, Fraction=Constant() {r=1})
  n37 = ScalarParameter() {defaultValue=0.2, group=Galaxy - Stars, parameterName=Galaxy.Stars_NebulaMask_Power, sortPriority=3}
  NamedRerouteDeclaration(None=MaterialFunctionCall(
    In=Power(Base=@n36, Exp=@n37)
    Contrast=ScalarParameter() {defaultValue=1, group=Galaxy - Stars, parameterName=Galaxy.Stars_NebulaMask_Contrast, sortPriority=4}
   ) {materialFunction=&CheapContrast}.Result) {name=StarMask, nodeColor=<rgba 0.9922 1 0 1>}
  n38 = NamedRerouteUsage()
  n39 = NamedRerouteUsage()
  n40 = VectorParameter() {group=Galaxy - General Parameters, parameterName=Galaxy.TOD_Morning, sortPriority=50, defaultValue=<rgba 1 1 1 1>}
  n41 = VectorParameter() {group=Galaxy - General Parameters, parameterName=Galaxy.TOD_Day, sortPriority=51, defaultValue=<rgba 1 1 1 1>}
  n42 = VectorParameter() {group=Galaxy - General Parameters, parameterName=Galaxy.TOD_Evening, sortPriority=52, defaultValue=<rgba 1 1 1 1>}
  n43 = VectorParameter() {group=Galaxy - General Parameters, parameterName=Galaxy.TOD_Night, sortPriority=53, defaultValue=<rgba 1 1 1 1>}
  n44 = Multiply(
   A=NamedRerouteUsage()
   B=MaterialFunctionCall(Color_Morning=@n40.RGB, Color_Day=@n41.RGB, Color_Evening=@n42.RGB, Color_Night=@n43.RGB) {materialFunction=&Color_Time_of_Day}.Result
  )
  n45 = Multiply(
   A=Saturate(None=LinearInterpolate(A=@n38, Alpha=@n39) {constB=0})
   B=Multiply(A=@n44, B=ScalarParameter() {defaultValue=25, group=Galaxy - Stars, parameterName=Galaxy.Stars_GlobalBrightness, sortPriority=1}) {constB=20}
  )
  n46 = ScalarParameter() {defaultValue=0.5, group=Galaxy - Stars, parameterName=Galaxy.Stars_Desaturation, sortPriority=2}
  NamedRerouteDeclaration(None=Add(A=Desaturation(None=@n45, Fraction=@n46), B=@n44)) {name=GalaxiesAndStars, nodeColor=<rgba 0 1 0.9062 1>}
  n47 = GetMaterialAttributes(None=NamedRerouteUsage()) {attributeGetTypes [
    #4
    #2
    #3
    #5
    #1
   ]}
  n48 = ScalarParameter() {group=Galaxy - General Parameters, parameterName=Galaxy.BaseMaterial_SMR_Override}
  NamedRerouteDeclaration(None=LinearInterpolate(
    A=@n47.Roughness
    B=ScalarParameter() {defaultValue=0.2, group=Galaxy - General Parameters, parameterName=Galaxy.Roughnes_Override, sortPriority=30}
    Alpha=@n48
   ) {constB=0}) {name=OUT_Roughness, nodeColor=<rgba 0 0.4844 1 1>}
  NamedRerouteDeclaration(None=LinearInterpolate(
    A=@n47.Metallic
    B=ScalarParameter() {defaultValue=0.1, group=Galaxy - General Parameters, parameterName=Galaxy.Metallic_Override, sortPriority=30}
    Alpha=@n48
   ) {constB=0}) {name=OUT_Metallic, nodeColor=<rgba 0 0.4844 1 1>}
  NamedRerouteDeclaration(None=LinearInterpolate(
    A=@n47.Specular
    B=ScalarParameter() {defaultValue=0.5, group=Galaxy - General Parameters, parameterName=Galaxy.Specular_Override, sortPriority=30}
    Alpha=@n48
   ) {constB=0}) {name=OUT_Spec, nodeColor=<rgba 0 0.4844 1 1>}
  NamedRerouteDeclaration(None=@n47.Emissive Color) {name=IN_Emissive, nodeColor=<rgba 0.4766 1 0 1>}
  n49 = VectorParameter() {group=Galaxy - General Parameters, parameterName=Galaxy.BaseColor_Override, sortPriority=30, defaultValue=<rgba 0 0 0 1>}
  NamedRerouteDeclaration(None=LinearInterpolate(A=@n47.Base Color, B=@n49.RGB, Alpha=ScalarParameter() {group=Galaxy - General Parameters, parameterName=Galaxy.BaseMaterial_Color_Override}) {constB=0}) {name=OUT_BaseColor, nodeColor=<rgba 1 0 0.6562 1>}
  NamedRerouteDeclaration(None=ChannelMaskParameter(None=NamedRerouteUsage()) {group=Galaxy - General Parameters, parameterName=Galaxy.MaskChannel, sortPriority=0}.R) {name=GalaxyMask, nodeColor=<rgba 0 1 0.6953 1>}
  n50 = VectorParameter() {group=Galaxy - Stars, parameterName=Glaxy.Stars_Speed_Layer2, sortPriority=23, defaultValue=<rgba -0.001 -0.01 0 1>}
  MaterialExpressionMaterialFunctionCall_4 = MaterialFunctionCall(
   Texture=NamedRerouteUsage()
   Depth=ScalarParameter() {defaultValue=150, group=Galaxy - Stars, parameterName=Galaxy.Stars_Depth_Layer2, sortPriority=20}
   Tile=ScalarParameter() {defaultValue=3, group=Galaxy - Stars, parameterName=Galaxy.Stars_Tile_Layer2, sortPriority=21}
   BlendContrast=ScalarParameter() {defaultValue=1, group=Galaxy - Stars, parameterName=Galaxy.Stars_BlendContrast_Layer2, sortPriority=22}
   SpeedXY=ComponentMask(None=@n50.RGB) {g=true, r=true}
   ActorOffset=NamedRerouteUsage()
   bUseDitherSingleSample=StaticBool() {value=true}
  ) {materialFunction=&MF_TriplanarVirtualBox}
  n51 = Convert(None=@MaterialExpressionMaterialFunctionCall_4.Result) {
   convertInputs [{
     type=Vector3
     expressionInput {
      expression=&self:MaterialExpressionMaterialFunctionCall_4
      inputName=None
      mask=0
      maskA=0
      maskB=0
      maskG=0
      maskR=0
      outputIndex=0
     }
    }]
   convertOutputs [{type=Scalar}, {type=Scalar}, {type=Scalar}]
  }
  NamedRerouteDeclaration(None=@n51) {name=FarStarsR, nodeColor=<rgba 0 1 0.2031 1>}
  NamedRerouteDeclaration(None=@n51) {name=FarStarsG, nodeColor=<rgba 1 0 0.1172 1>}
  NamedRerouteDeclaration(None=@MaterialExpressionMaterialFunctionCall_4.DitherUVs) {name=FarStarsUVW, nodeColor=<rgba 1 0 0.8906 1>}
  n52 = VectorParameter() {group=Galaxy - Stars, parameterName=Glaxy.Stars_Speed_Layer1, sortPriority=13, defaultValue=<rgba -0.001 -0.01 0 1>}
  MaterialExpressionMaterialFunctionCall_18 = MaterialFunctionCall(
   Texture=NamedRerouteUsage()
   Depth=ScalarParameter() {defaultValue=75, group=Galaxy - Stars, parameterName=Galaxy.Stars_Depth_Layer1, sortPriority=10}
   Tile=ScalarParameter() {defaultValue=1.5, group=Galaxy - Stars, parameterName=Galaxy.Stars_Tile_Layer1, sortPriority=11}
   BlendContrast=ScalarParameter() {defaultValue=1, group=Galaxy - Stars, parameterName=Galaxy.Stars_BlendContrast_Layer1, sortPriority=12}
   SpeedXY=ComponentMask(None=@n52.RGB) {g=true, r=true}
   ActorOffset=NamedRerouteUsage()
   bUseDitherSingleSample=StaticBool() {value=true}
  ) {materialFunction=&MF_TriplanarVirtualBox}
  n53 = Convert(None=@MaterialExpressionMaterialFunctionCall_18.Result) {
   convertInputs [{
     type=Vector3
     expressionInput {
      expression=&self:MaterialExpressionMaterialFunctionCall_18
      inputName=None
      mask=0
      maskA=0
      maskB=0
      maskG=0
      maskR=0
      outputIndex=0
     }
    }]
   convertOutputs [{type=Scalar}, {type=Scalar}, {type=Scalar}]
  }
  NamedRerouteDeclaration(None=@n53) {name=CloseStarsR, nodeColor=<rgba 0 0.3438 1 1>}
  n54 = TextureObjectParameter() {
   group=Galaxy - Stars
   parameterName=Galaxy.StarsTexture
   samplerType=SAMPLERTYPE_LinearColor
   sortPriority=0
   texture=/Game/Global/Materials/Character_Parent_2023/Galaxy_Parent_2025/Textures/T_GalaxyStars
  }
  NamedRerouteDeclaration(None=@n54) {name=StarTex, nodeColor=<rgba 1 0.1641 0 1>}
  NamedRerouteDeclaration(None=@n53) {name=CloseStarsG, nodeColor=<rgba 0 0.1562 1 1>}
  NamedRerouteDeclaration(None=@MaterialExpressionMaterialFunctionCall_18.DitherUVs) {name=CloseStarsUVW, nodeColor=<rgba 1 0 0.7969 1>}
  NamedRerouteDeclaration(None=NamedRerouteDeclaration(None=@n53) {name=CloseStarsB, nodeColor=<rgba 0.4766 1 0 1>}) {name=CloseStarsRandom, nodeColor=<rgba 1 0.3984 0 1>}
  n55 = GetMaterialAttributes(None=MaterialFunctionCall!(Material=@n5.Result, Tech Art Mask=@n5.Result) {materialFunction=/CosmeticCompanions/Materials/MaterialFunctions/MF_FN_Companion_Custom_Color}.Base Color) {attributeGetTypes [#1]}
  NamedRerouteDeclaration(None=SetMaterialAttributes!(MaterialAttributes=@n55.MaterialAttributes, Emissive Color=@n55.MaterialAttributes) {attributeSetTypes [#5]}) {name=IN_Material, nodeColor=<rgba 1 0 0.9609 1>}
  NamedRerouteDeclaration(None=NamedRerouteDeclaration(None=@n51) {name=FarStarsB, nodeColor=<rgba 0 0.4375 1 1>}) {name=FarStarsRandom, nodeColor=<rgba 1 0.3984 0 1>}
  NamedRerouteDeclaration(None=BlendMaterialAttributes(A=NamedRerouteUsage(), B=NamedRerouteUsage(), Alpha=NamedRerouteUsage())) {name=GalaxyMaterial, nodeColor=<rgba 0 1 0.2266 1>}
  n56 = Multiply(
   A=Fresnel(ExponentIn=ScalarParameter() {defaultValue=3, group=Galaxy, parameterName=EmissiveFresnelExponent, sortPriority=90}) {baseReflectFraction=0}
   B=ScalarParameter() {defaultValue=1, group=Galaxy, parameterName=EmissiveFresnelIntensity, sortPriority=92}
  )
  NamedRerouteDeclaration(None=Multiply(
    A=NamedRerouteDeclaration(None=@n56) {name=FresnelMask, nodeColor=<rgba 0 0.7891 1 1>}
    B=VectorParameter() {group=Color Customization, parameterName=Primary Color, defaultValue=<rgba 1 1 1 1>}.RGB
   )) {name=Fresnel, nodeColor=<rgba 0.8047 0 1 1>}
 ]
}
