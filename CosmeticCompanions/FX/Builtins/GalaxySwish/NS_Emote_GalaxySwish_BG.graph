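# Niagara system for the background of the Galaxy Swish companion emote, staging timed falling stars, a glowing main star, a smoke ring and a constellation.
# MainStar follows a baked spline mesh; it and FallingStars feed secondary trail emitters via particle reads.
# ConvergingStars blend from animated shapes onto skeletal mesh sample points.
# Constellation meshes link indexed star pairs through a shared position array, stretching to line length and flashing on loop-age curves.
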NiagaraSystem /CosmeticCompanions/FX/Builtins/GalaxySwish/NS_Emote_GalaxySwish_BG
dependencies {
 usedDataInterfaces [
  &NiagaraDataInterfaceCurve
  &NiagaraDataInterfaceColorCurve
  &NiagaraDataInterfaceParticleRead
  &NiagaraDataInterfaceArrayInt32
  &NiagaraDataInterfaceStaticMesh
  &NiagaraDataInterfaceArrayPosition
  &NiagaraDataInterfaceSkeletalMesh
 ]
 usedDynamicInputs []
 usedModules [
  &self:SystemSpawnScript.SystemScriptSource.SystemScriptGraph.NiagaraNodeAssignment_1.SetVariables_#1
  &SystemState
  &self:SystemSpawnScript.SystemScriptSource.SystemScriptGraph.NiagaraNodeAssignment_0.SetVariables_#2
 ]
}
emitters {
 Constellation {
  emitterUpdateScript [
   &EmitterState(
    Inactive Response=::NewEnumerator1
    Life Cycle Mode=::NewEnumerator1
    Loop Behavior=::NewEnumerator1
    Loop Delay=10.35
    Loop Duration=4.4
    UseLoopDelay=true
   )
   &SpawnBurst_Instantaneous(Spawn Count=13, Use Spawn Probability=false, UseLoopCountLimit=false)
  ]
  particleSpawnScript [
   &FN_SetSpawnGroup(Number of Spawn Groups=13)
   &InitializeParticle(
    Color=<rgba 5 5 5 1>
    Lifetime=4.4
    Mesh Scale Mode=::NewEnumerator3
    Position Offset=<0 0 -80>
    UsePositionOffset=true
    Write Material Random=true
   )
   /SetVariables(Particles.InitialIndex:NiagaraInt32=$Emitter.SpawnGroup) {enabled=false, isSetParameters=true}
   /SetVariables(
    Particles.IndexA:NiagaraInt32=&SelectIntFromArray(
     Direct Array Index=&ReturnExecIndex()
     Int Selection Array=DataInterface() {propertyValues {IntData [0,0,1,2,3,4,11,12,12,12,12,8,9]}}
    )
    Particles.IndexB:NiagaraInt32=&SelectIntFromArray(
     Direct Array Index=&ReturnExecIndex()
     Int Selection Array=DataInterface() {propertyValues {IntData [1,2,3,4,11,11,12,5,6,7,8,9,10]}}
    )
   ) {isSetParameters=true}
   &CameraOffset(Camera Offset Amount=15)
  ]
  particleUpdateScript [
   &ParticleState()
   &self:ScratchModule_01(
    Index:NiagaraInt32=$Particles.IndexA
    Index02:NiagaraInt32=$Particles.IndexB
    Position Array:NiagaraDataInterfaceArrayPosition=$System.Position Array
   ) {module=ScratchModule_01}
   &ScaleMeshSize(Scale Factor=&MakeVector(X=&Divide_Float(A=$Particles.LightningLength, B=100), Y=1, Z=1))
   &DynamicMaterialParameters(
    Index 0 Param 1=15
    Index 0 Param 2=1
    Module.Index0Float0Write=true
    Module.Index0Float1Write=true
    Module.Index0Float2Write=true
    Module.Index0Float3Write=true
   )
   &ScaleColor(
    Scale Alpha=&self:Merge_Helper_Float_Switch_15(
     0:NiagaraFloat=&FloatFromCurve(FloatCurve=DataInterface() {propertyValues {Curve {keys [<at 0.05 0>,<at 0.15 1>,<at 0.925 1>,<at 1 0>]}}})
     1:NiagaraFloat=&FloatFromCurve(FloatCurve=DataInterface() {propertyValues {Curve {keys [<at 0.175 0>,<at 0.25 1>,<at 0.925 1>,<at 1 0>]}}})
     2:NiagaraFloat=&FloatFromCurve(FloatCurve=DataInterface() {propertyValues {Curve {keys [<at 0.125 0>,<at 0.2 1>,<at 0.925 1>,<at 1 0>]}}})
     3:NiagaraFloat=&FloatFromCurve(FloatCurve=DataInterface() {propertyValues {Curve {keys [<at 0.05 0>,<at 0.15 1>,<at 0.925 1>,<at 1 0>]}}})
     4:NiagaraFloat=&FloatFromCurve(FloatCurve=DataInterface() {propertyValues {Curve {keys [<at 0.175 0>,<at 0.25 1>,<at 0.925 1>,<at 1 0>]}}})
     5:NiagaraFloat=&FloatFromCurve(FloatCurve=DataInterface() {propertyValues {Curve {keys [<at 0.125 0>,<at 0.2 1>,<at 0.925 1>,<at 1 0>]}}})
     6:NiagaraFloat=&FloatFromCurve(FloatCurve=DataInterface() {propertyValues {Curve {keys [<at 0.05 0>,<at 0.15 1>,<at 0.925 1>,<at 1 0>]}}})
     7:NiagaraFloat=&FloatFromCurve(FloatCurve=DataInterface() {propertyValues {Curve {keys [<at 0.175 0>,<at 0.25 1>,<at 0.925 1>,<at 1 0>]}}})
     8:NiagaraFloat=&FloatFromCurve(FloatCurve=DataInterface() {propertyValues {Curve {keys [<at 0.125 0>,<at 0.2 1>,<at 0.925 1>,<at 1 0>]}}})
     9:NiagaraFloat=&FloatFromCurve(FloatCurve=DataInterface() {propertyValues {Curve {keys [<at 0.05 0>,<at 0.15 1>,<at 0.925 1>,<at 1 0>]}}})
     10:NiagaraFloat=&FloatFromCurve(FloatCurve=DataInterface() {propertyValues {Curve {keys [<at 0.175 0>,<at 0.25 1>,<at 0.925 1>,<at 1 0>]}}})
     11:NiagaraFloat=&FloatFromCurve(FloatCurve=DataInterface() {propertyValues {Curve {keys [<at 0.125 0>,<at 0.2 1>,<at 0.925 1>,<at 1 0>]}}})
     12:NiagaraFloat=&FloatFromCurve(FloatCurve=DataInterface() {propertyValues {Curve {keys [<at 0.05 0>,<at 0.15 1>,<at 0.925 1>,<at 1 0>]}}})
     13:NiagaraFloat=?
     14:NiagaraFloat=?
     Index:NiagaraInt32=&SpawnGroup()
     Num:NiagaraInt32=11
    )
    Scale RGB=&VectorFromFloat(Value=&FloatFromCurve(
      FloatCurve=DataInterface() {propertyValues {Curve {keys [
          <at 0 1>
          <at 1.25 1>
          <at 1.35 8>
          <at 1.45 1>
          <at 2.6 1>
          <at 2.7 8>
          <at 2.85 1>
          {arriveTangent=0.139, interpMode=RCIM_Cubic, leaveTangent=0.139, tangentMode=RCTM_User, time=3.4, value=1.0195}
          {arriveTangent=12.2005, interpMode=RCIM_Cubic, leaveTangent=12.2005, time=4.07, value=8}
          {arriveTangent=-15.1515, interpMode=RCIM_Cubic, leaveTangent=-15.1515, time=4.3, value=12}
          <at 4.4 3>
         ]}}}
      CurveIndex=$Emitter.LoopedAge
     ))
    ScaleA=true
    ScaleRGB=true
   )
  ]
  rendererClasses [&NiagaraMeshRendererProperties]
  renderers [{rendererClass=&NiagaraMeshRendererProperties, rendererIndex=0}]
 }
 ConvergingStars {
  emitterUpdateScript [
   &EmitterState(
    Inactive Response=::NewEnumerator1
    Life Cycle Mode=::NewEnumerator1
    Loop Behavior=::NewEnumerator1
    Loop Delay=6
    Loop Duration=8.75
    UseLoopDelay=true
   )
   &SpawnBurst_Instantaneous(Spawn Count=18, Use Spawn Probability=false, UseLoopCountLimit=false)
  ]
  particleSpawnScript [
   &FN_SetSpawnGroup(Number of Spawn Groups=13)
   &InitializeParticle(
    Color=&MaskLinearColorBySpawnGroup(
     Masked Linear Color=<rgba 0.8594 4.8025 5 1>
     Passthrough Linear Color=&RandomRangeLinearColor(Minimum=<rgba 0.3945 0.6791 0.7 1>, Maximum=<rgba 0.3438 1.921 2 1>)
     Spawn Group Mask=11
    )
    Lifetime Max=8.75
    Lifetime Min=8.5
    Lifetime Mode=::NewEnumerator1
    Position Offset=<0 0 40>
    Sprite Rotation Mode=::NewEnumerator2
    Sprite Size Mode=::NewEnumerator3
    Uniform Sprite Size=&MaskFloatBySpawnGroup(Masked Float=42, Passthrough Float=&RandomRangeFloat(Minimum=24, Maximum=48), Spawn Group Mask=11)
    UsePositionOffset=false
    Write Material Random=true
   )
   /SetVariables(Particles.InitialIndex:NiagaraInt32=$Particles.SpawnGroup) {isSetParameters=true}
   &CameraOffset(Camera Offset Amount=25)
  ]
  particleUpdateScript [
   &ParticleState()
   &SpriteRotationRate(Rotation Rate=&RandomRangeFloat(Minimum=-25, Maximum=25))
   &ScaleSpriteSize(Uniform Curve Sprite Scale=DataInterface() {propertyValues {Curve {keys [
        <at 0 0 RCIM_Cubic>
        {arriveTangent=10, interpMode=RCIM_Cubic, leaveTangent=10, time=0.056, value=1.2}
        {arriveTangent=0.2647, interpMode=RCIM_Cubic, leaveTangent=0.2647, tangentMode=RCTM_User, time=0.1, value=1}
        <at 1 1>
       ]}}})
   &ShapeLocation(
    Disc Coverage=1
    MaskSpawnGroup=false
    Offset=&MakeVector(
     X=0
     Y=-100
     Z=&Add_Float(
      A=-30.5189
      B=&FloatFromCurve(FloatCurve=DataInterface() {propertyValues {Curve {keys [<at 0 1>,<at 2.5 0>]}}}, CurveIndex=$Emitter.LoopedAge, Scale Curve=50)
     )
    )
    Offset Coordinate Space=::Local
    Offset Mode=::NewEnumerator0
    Override Seed=false
    Ring Radius=&FloatFromCurve(FloatCurve=DataInterface() {propertyValues {Curve {keys [<at 0 1.6>,<at 2.5 1>]}}}, CurveIndex=$Emitter.LoopedAge, Scale Curve=80)
    Rotation Mode=::NewEnumerator2
    Shape Primitive=::NewEnumerator4
    UseOwnerScale=true
    Yaw / Pitch / Roll=&MakeVector(
     X=0
     Y=&FloatFromCurve(FloatCurve=DataInterface() {propertyValues {Curve {keys [<at 0 1>,<at 0.5 0.8>]}}}, CurveIndex=$Emitter.NormalizedLoopAge, Scale Curve=-1)
     Z=0.25
    )
   )
   &ShapeLocation(
    Handle Radius:NiagaraFloat=&FloatFromCurve(
     FloatCurve=DataInterface() {propertyValues {Curve {keys [<at 2.5 1 RCIM_Cubic>,<at 4.3 0 RCIM_Cubic>]}}}
     CurveIndex=$Emitter.LoopedAge
     Scale Curve=8
    )
    Large Radius=&FloatFromCurve(
     FloatCurve=DataInterface() {propertyValues {Curve {keys [<at 2.5 1 RCIM_Cubic>,<at 4.3 0 RCIM_Cubic>]}}}
     CurveIndex=$Emitter.LoopedAge
     Scale Curve=24
    )
    MaskSpawnGroup=false
    Non Uniform Scale=<1 1 1.85>
    Override Seed=false
    Rotation Angle:NiagaraFloat=&FloatFromCurve(FloatCurve=DataInterface() {propertyValues {Curve {keys [<at 0 15>,<at 1 0>]}}}, CurveIndex=$Emitter.NormalizedLoopAge, Scale Curve=-1)
    Rotation Axis=<0 1 0>
    Rotation Mode=::NewEnumerator1
    Shape Primitive=::NewEnumerator3
    Surface Distribution=0.7024
    UseOwnerScale=true
   ) {module=ShapeLocation001}
   &SkeletalMeshLocation(
    Bone / Socket Index=&SpawnGroup()
    MaskSpawnGroup=false
    Mesh Sampling Type=::NewEnumerator5
    Position Sampling=::NewEnumerator2
    Skeletal Mesh=$System.Skeletal Mesh Reader
    Socket Sampling Mode=::NewEnumerator5
   )
   /SetVariables(Particles.Position=&AddVectorToPosition(
     Position=&MakeVector(
      X=0
      Y=&FloatFromCurve(
       FloatCurve=DataInterface() {propertyValues {Curve {keys [<at 0 0.8 RCIM_Cubic>, {arriveTangent=0.0465, interpMode=RCIM_Cubic, leaveTangent=0.0465, time=2.5, value=1}, <at 4.3 1>, <at 4.35 0>]}}}
       CurveIndex=$Emitter.LoopedAge
       Scale Curve=67
      )
      Z=&FloatFromCurve(FloatCurve=DataInterface() {propertyValues {Curve {keys [<at 4.3 1>,<at 4.35 0>]}}}, CurveIndex=$Emitter.LoopedAge, Scale Curve=70)
     )
     Vector=&Lerp_Vector(
      A=&Lerp_Vector(
       A=Linked() {linkedVariable {name=Particles.ShapeLocation.ShapePosition, type=NiagaraPosition}}
       B=Linked() {linkedVariable {name=Particles.ShapeLocation001.ShapePosition, type=NiagaraPosition}}
       Alpha=&FloatFromCurve(
        FloatCurve=DataInterface() {propertyValues {Curve {
           defaultValue=1
           keys [
            <at 0 0 RCIM_Cubic>
            {arriveTangent=0.0414, interpMode=RCIM_Cubic, leaveTangent=0.0414, tangentMode=RCTM_User, time=0.5, value=0}
            <at 2 1 RCIM_Cubic>
           ]
          }}}
        CurveIndex=$Emitter.LoopedAge
       )
      )
      B=Linked() {linkedVariable {name=Particles.SkeletalMeshLocation.SampledPosition, type=NiagaraPosition}}
      Alpha=&FloatFromCurve(
       FloatCurve=DataInterface() {propertyValues {Curve {defaultValue=1, keys [{tangentMode=RCTM_User, time=0, value=0}, {tangentMode=RCTM_User, time=4.3, value=0}, <at 4.35 1>]}}}
       CurveIndex=$Emitter.LoopedAge
      )
     )
    )) {isSetParameters=true}
   &self:ScratchModule(
    FlareIndex:NiagaraFloat=&MakeFloatFromInt(Int=$Particles.InitialIndex)
    Position Array:NiagaraDataInterfaceArrayPosition=$System.Position Array
   ) {module=ScratchModule}
   &ScaleColor(
    Scale Alpha=&FloatFromCurve(FloatCurve=DataInterface() {propertyValues {Curve {keys [<at 0 0 RCIM_Cubic>, {arriveTangent=0.2647, interpMode=RCIM_Cubic, leaveTangent=0.2647, tangentMode=RCTM_User, time=0.1, value=1}, <at 1 1>]}}})
    Scale RGB=&VectorFromFloat(Value=&FloatFromCurve(FloatCurve=DataInterface() {propertyValues {Curve {keys [<at 0 1>,<at 0.05 3>,<at 0.1 1>,<at 1 1>]}}}))
    ScaleA=true
    ScaleRGB=true
   )
  ]
  rendererClasses [&NiagaraSpriteRendererProperties]
  renderers [{rendererClass=&NiagaraSpriteRendererProperties}]
 }
 FallingStars {
  emitterUpdateScript [
   &EmitterState(
    Inactive Response=::NewEnumerator1
    Life Cycle Mode=::NewEnumerator1
    Loop Behavior=::NewEnumerator1
    Loop Delay=3
    Loop Duration=12.8
    UseLoopDelay=false
   )
   &SpawnBurst_Instantaneous(Spawn Count=24, Use Spawn Probability=false, UseLoopCountLimit=false) {enabled=false}
   &SpawnRate(SpawnRate=4, Use Spawn Probability=false)
  ]
  particleSpawnScript [
   &InitializeParticle(
    Color Maximum=<rgba 0.3438 1.921 2 1>
    Color Minimum=<rgba 0.3945 0.6791 0.7 1>
    Color Mode=::NewEnumerator2
    Lifetime Max=5
    Lifetime Min=2
    Lifetime Mode=::NewEnumerator1
    Position Offset=<0 -45 80>
    Sprite Rotation Mode=::NewEnumerator2
    Sprite Size Mode=::NewEnumerator1
    Sprite UV Mode=::NewEnumerator3
    Uniform Sprite Size Max=96
    Uniform Sprite Size Min=24
    UsePositionOffset=true
    Write Material Random=true
   )
   &ShapeLocation(
    Disc Coverage=0.3
    MaskSpawnGroup=false
    Override Seed=false
    Ring Radius=101.5184
    Rotation Angle=-0.125
    Rotation Axis=<0 1 0>
    Rotation Mode=::NewEnumerator1
    Shape Primitive=::NewEnumerator4
    U Distribution=0.75
    UseOwnerScale=true
   )
   &AddVelocity(
    Module.Use Falloff Distance=false
    Override Seed=false
    Velocity Mode=::NewEnumerator1
    Velocity Speed=&RandomRangeFloat(Minimum=25, Maximum=200)
   )
   &CameraOffset(Camera Offset Amount=&RandomRangeFloat(Minimum=-5, Maximum=15))
  ]
  particleUpdateScript [
   &ParticleState()
   &SpriteRotationRate(Rotation Rate=&UniformAOrBFloat(A=-250, B=250, Evaluation Type=::NewEnumerator0, Recalculate Random Each Loop=true, Override Seed=false))
   &CurlNoiseForce(
    Bias Noise=false
    Noise Frequency=20
    Noise Strength=&RandomRangeFloat(Minimum=45, Maximum=120)
    Pan Noise=false
   )
   &ScaleSpriteSize(Scale Factor=&Vector2DFromFloat(Value=&FloatFromCurve(FloatCurve=DataInterface() {propertyValues {Curve {keys [
          <at 0 0 RCIM_Cubic>
          <at 0.112 1.0173>
          {arriveTangent=-1.142, interpMode=RCIM_Cubic, leaveTangent=-1.142, time=0.7001, value=0.872}
          <at 1.0011 0.002>
         ]}}})))
   &ScaleColor(
    Scale Alpha=&FloatFromCurve(FloatCurve=DataInterface() {propertyValues {Curve {keys [<at 0 0 RCIM_Cubic>,<at 0.5 1 RCIM_Cubic>,<at 1 0 RCIM_Cubic>]}}})
    ScaleA=true
    ScaleRGB=true
   )
   &Drag(Drag=&RandomRangeFloat(Minimum=2, Maximum=5), Use Linear Drag=true, Use Rotational Drag=true)
   &GravityForce(Gravity=&RandomRangeVector(Minimum=<0 0 -200>, Maximum=<0 0 -480>))
   &SolveForcesAndVelocity(Clamp Velocity=false, Limit Acceleration=false)
  ]
  rendererClasses [&NiagaraSpriteRendererProperties]
  renderers [{rendererClass=&NiagaraSpriteRendererProperties}]
 }
 FallingStars_Secondary {
  emitterUpdateScript [
   &EmitterState(Inactive Response=::NewEnumerator1, Life Cycle Mode=::NewEnumerator1, Loop Behavior=::NewEnumerator1, Loop Duration=14, UseLoopDelay=false)
   &SpawnParticlesFromOtherEmitter(
    Attribute Reader=DataInterface() {propertyValues {EmitterBinding {bindingMode=Other, emitterName=FallingStars}}}
    Use Maximum Age=false
    Use Minimum Age=false
   )
  ]
  particleSpawnScript [
   &InitializeParticle(
    Color=<rgba 1 1 1 0.3>
    Lifetime Max=0.8
    Lifetime Min=0.1
    Lifetime Mode=::NewEnumerator1
    Position Offset=<0 -45 80>
    UsePositionOffset=true
    Write Material Random=true
   )
   &SampleParticlesFromOtherEmitter(
    Applied Velocity Scale=0.2
    Color Sampling=::NewEnumerator0
    Sprite Size Sampling:ENiagara_DirectReadSamplingMode=::NewEnumerator1
    Velocity Sampling=::NewEnumerator0
   )
   &ShapeLocation(MaskSpawnGroup=false, Override Seed=false, Sphere Radius=6, UseOwnerScale=true)
   /SetVariables(Particles.SpriteSize=&Multiply_Vector2D_ByFloat(Float=0.3, Vector2D=$Output.SampleParticlesFromOtherEmitter.SampledSpriteSize)) {isSetParameters=true}
  ]
  particleUpdateScript [
   &ParticleState()
   &CurlNoiseForce(
    Bias Noise=false
    Noise Frequency=25
    Noise Strength=&RandomRangeFloat(Minimum=25, Maximum=40)
    Pan Noise=false
   )
   &SolveForcesAndVelocity(Clamp Velocity=false, Limit Acceleration=false)
   &ScaleColor(
    Scale Alpha=&FloatFromCurve(FloatCurve=DataInterface() {propertyValues {Curve {
        defaultValue=1
        keys [
         {arriveTangent=9.0725, interpMode=RCIM_Cubic, leaveTangent=9.0725, tangentMode=RCTM_User, time=0, value=0}
         {arriveTangent=-2.5249, interpMode=RCIM_Cubic, leaveTangent=-2.5249, tangentMode=RCTM_User, time=0.1408, value=0.9847}
         {arriveTangent=-0.0405, interpMode=RCIM_Cubic, leaveTangent=-0.0405, tangentMode=RCTM_User, time=1, value=0}
        ]
       }}})
    ScaleA=true
    ScaleRGB=true
   )
  ]
  rendererClasses [&NiagaraSpriteRendererProperties]
  renderers [{rendererClass=&NiagaraSpriteRendererProperties}]
 }
 MainStar {
  emitterUpdateScript [
   &EmitterState(
    Inactive Response=::NewEnumerator1
    Life Cycle Mode=::NewEnumerator1
    Loop Behavior=::NewEnumerator1
    Loop Delay=1.7
    Loop Duration=14
    UseLoopDelay=true
   )
   &SpawnBurst_Instantaneous(Use Spawn Probability=false, UseLoopCountLimit=false)
   &SpawnBurst_Instantaneous(Spawn Group=1, Spawn Time=2.3, Use Spawn Probability=false, UseLoopCountLimit=false) {module=SpawnBurst_Instantaneous001}
   &SpawnBurst_Instantaneous(Spawn Group=2, Spawn Time=2.42, Use Spawn Probability=false, UseLoopCountLimit=false) {module=SpawnBurst_Instantaneous002}
   &SpawnBurst_Instantaneous(Spawn Group=3, Spawn Time=8.6, Use Spawn Probability=false, UseLoopCountLimit=false) {module=SpawnBurst_Instantaneous003}
   &SpawnBurst_Instantaneous(Spawn Group=4, Spawn Time=13, Use Spawn Probability=false, UseLoopCountLimit=false) {module=SpawnBurst_Instantaneous005}
   &SpawnBurst_Instantaneous(Spawn Group=3, Spawn Time=12.12, Use Spawn Probability=false, UseLoopCountLimit=false) {enabled=false, module=SpawnBurst_Instantaneous004}
   /SetVariables(Emitter.StarGlowColor:LinearColor=&Multiply_LinearColorByFloat(
     Linear Color=<rgba 2 0.7584 0.1078 1>
     Float=&Waveform("Waveform [1]"=::NewEnumerator2, "[1] Frequency"=3, "[1] Amplitude Min/Max"=<1.2 6>)
    )) {isSetParameters=true}
  ]
  particleSpawnScript [
   /SetVariables(Particles.VisibilityTag=&SpawnGroup()) {isSetParameters=true}
   &InitializeParticle(
    Color=<rgba 5 2.2684 0 0.2087>
    Lifetime=&MergeHelper_FloatSwitch(0=13.2, 1=10.9, 2=0.64, 3=0.64, 4=0.3, Index=&SpawnGroup(), Num=::NewEnumerator3)
    Mesh Scale Mode=::NewEnumerator3
    Mesh Uniform Scale=&MergeHelper_FloatSwitch(0=0.25, 1=0, 2=0, 3=0, 4=1.7, Index=&SpawnGroup(), Num=::NewEnumerator3)
    Sprite Size Mode=::NewEnumerator3
    Uniform Sprite Size=31.8552
    UsePositionOffset=false
    Write Material Random=true
   )
  ]
  particleUpdateScript [
   &SolveForcesAndVelocity(Clamp Velocity=false, Limit Acceleration=false)
   &ParticleState()
   &ScaleSpriteSize(Uniform Curve Sprite Scale=DataInterface() {propertyValues {Curve {keys [{tangentMode=RCTM_User, time=0, value=1}, <at 0.65 1>, <at 0.66 8>, {tangentMode=RCTM_User, time=0.67, value=0}]}}})
   &ScaleMeshSize(Scale Factor=&VectorFromFloat(Value=&FloatFromCurve(FloatCurve=DataInterface() {propertyValues {Curve {keys [
          <at 0 0 RCIM_Cubic>
          {arriveTangent=0.1775, interpMode=RCIM_Cubic, leaveTangent=0.1775, tangentMode=RCTM_User, time=0.05, value=1}
          <at 0.55 1>
          <at 0.64 0.0471>
         ]}}})))
   /SetVariables(Particles.0Color:LinearColor=&Multiply_LinearColorByFloat(
     Linear Color=$Particles.Color
     Float=&FloatFromCurve(FloatCurve=DataInterface() {propertyValues {Curve {keys [<at 0 1>,<at 0.625 1>,<at 0.66 0.6>,<at 1 0>]}}})
    )) {isSetParameters=true}
   /SetVariables(
    Particles.1Color:LinearColor=&MultiplyLinearColor(
     Linear Color=<rgba 8 4.4507 0.1039 0.4>
     Scale Factor=&MakeVector4FromVectorAndFloat(
      "Vector (RGB)"=<1 1 1>
      "Float (Alpha)"=&Waveform("Waveform [1]"=::NewEnumerator2, "[1] Frequency"=2, "[1] Compound Frequency"=1.5, "[1] Amplitude Min/Max"=<0.4 1>)
     )
    )
    Particles.1Size:Vector2f=&Multiply_Vector2D(
     A=&Vector2DFromFloat(Value=&Waveform("Waveform [1]"=::NewEnumerator0, "[1] Frequency"=2, "[1] Amplitude Min/Max"=<0.7 1>))
     B=&Vector2DFromFloat(Value=&FloatFromCurve(
       FloatCurve=DataInterface() {propertyValues {Curve {
          defaultValue=1
          keys [
           {arriveTangent=9.0725, interpMode=RCIM_Cubic, leaveTangent=9.0725, tangentMode=RCTM_User, time=0, value=0}
           {arriveTangent=-0.0052, interpMode=RCIM_Cubic, leaveTangent=-0.0052, tangentMode=RCTM_User, time=0.0225, value=1.2}
           {arriveTangent=-2.3034, interpMode=RCIM_Cubic, leaveTangent=-2.3034, tangentMode=RCTM_User, time=0.237, value=0.6384}
           {arriveTangent=-1.8886, interpMode=RCIM_Cubic, leaveTangent=-1.8886, time=0.45, value=1.0384}
           {arriveTangent=-26.0195, interpMode=RCIM_Cubic, leaveTangent=-26.0195, tangentMode=RCTM_User, time=0.575, value=0}
           {arriveTangent=-0.0868, interpMode=RCIM_Cubic, leaveTangent=-0.0868, tangentMode=RCTM_User, time=0.58, value=0}
           {arriveTangent=26.6666, interpMode=RCIM_Cubic, leaveTangent=26.6666, time=0.59, value=1.7}
           {arriveTangent=-48.5714, interpMode=RCIM_Cubic, leaveTangent=-48.5714, time=0.61, value=0.8}
           {arriveTangent=-0.3504, interpMode=RCIM_Cubic, leaveTangent=-0.3504, tangentMode=RCTM_User, time=0.625, value=0}
           {arriveTangent=0.1018, interpMode=RCIM_Cubic, leaveTangent=0.1018, tangentMode=RCTM_User, time=0.895, value=0}
           <at 0.915 0 RCIM_Cubic>
           {arriveTangent=0.0866, interpMode=RCIM_Cubic, leaveTangent=0.0866, tangentMode=RCTM_User, time=0.94, value=0}
           {arriveTangent=-0.0098, interpMode=RCIM_Cubic, leaveTangent=-0.0098, tangentMode=RCTM_User, time=0.974, value=0}
           <at 0.985 0 RCIM_Cubic>
           <at 1 0 RCIM_Cubic>
          ]
         }}}
       Scale Curve=125
      ))
    )
   ) {isSetParameters=true}
   /SetVariables(
    Particles.2CO=15
    Particles.2Color:LinearColor=&ColorFromCurve(ColorCurve=DataInterface() {propertyValues {
       AlphaCurve {keys [<at 0 0>,<at 0.2 1 RCIM_Cubic>,<at 1 0 RCIM_Cubic>]}
       BlueCurve {keys [<at 0 1.8595 RCIM_Cubic>,<at 1 0.6198>]}
       GreenCurve {keys [<at 0 7.2574 RCIM_Cubic>,<at 1 2.4191>]}
       RedCurve {keys [<at 0 9 RCIM_Cubic>,<at 1 3>]}
      }})
    Particles.2Pos:NiagaraPosition=&AddVectorToPosition(Position=$Particles.Position, Vector=<0 0 3.143>)
    Particles.2Size:Vector2f=&Multiply_Vector2D_ByFloat(
     Float=&FloatFromCurve(FloatCurve=DataInterface() {propertyValues {Curve {
         defaultValue=1
         keys [
          {arriveTangent=9.0725, interpMode=RCIM_Cubic, leaveTangent=9.0725, tangentMode=RCTM_User, time=0, value=0}
          {arriveTangent=-0.0604, interpMode=RCIM_Cubic, leaveTangent=-0.0604, tangentMode=RCTM_User, time=0.1408, value=0.9847}
         ]
        }}})
     Vector2D=<39.5204 33.8084>
    )
   ) {isSetParameters=true}
   /SetVariables(Particles.3Size:Vector2f=&Multiply_Vector2D_ByFloat(Float=1.598, Vector2D=$Particles.2Size)) {isSetParameters=true}
   /SetVariables(
    Particles.4Color:LinearColor=&MakeLinearColorFromVectorAndFloat(
     "Vector (RGB)"=&MakeVectorFromLinearColorRGB(Color=<rgba 12 7.3235 0 1>)
     "Float (Alpha)"=&FloatFromCurve(
      FloatCurve=DataInterface() {propertyValues {Curve {
         defaultValue=1
         keys [
          {arriveTangent=9.0725, interpMode=RCIM_Cubic, leaveTangent=9.0725, tangentMode=RCTM_User, time=0, value=0}
          {arriveTangent=-0.0481, interpMode=RCIM_Cubic, leaveTangent=-0.0481, tangentMode=RCTM_User, time=0.1408, value=0.9847}
          {arriveTangent=-0.0405, interpMode=RCIM_Cubic, leaveTangent=-0.0405, tangentMode=RCTM_User, time=1, value=0}
         ]
        }}}
      Scale Curve=0.3
     )
    )
    Particles.4RaysColor:LinearColor=&MakeLinearColorFromVectorAndFloat(
     "Vector (RGB)"=&MakeVectorFromLinearColorRGB(Color=<rgba 12 7.1744 0.8503 1>)
     "Float (Alpha)"=&FloatFromCurve(
      FloatCurve=DataInterface() {propertyValues {Curve {
         defaultValue=1
         keys [
          {arriveTangent=9.0725, interpMode=RCIM_Cubic, leaveTangent=9.0725, tangentMode=RCTM_User, time=0, value=0}
          {arriveTangent=-0.0481, interpMode=RCIM_Cubic, leaveTangent=-0.0481, tangentMode=RCTM_User, time=0.1408, value=0.9847}
          {arriveTangent=-0.0405, interpMode=RCIM_Cubic, leaveTangent=-0.0405, tangentMode=RCTM_User, time=1, value=0}
         ]
        }}}
      Scale Curve=0.05
     )
    )
    Particles.4Size:Vector2f=&Multiply_Vector2D_ByFloat(
     Float=&FloatFromCurve(FloatCurve=DataInterface() {propertyValues {Curve {
         defaultValue=0.2
         keys [
          {arriveTangent=4.1785, interpMode=RCIM_Cubic, leaveTangent=4.1785, tangentMode=RCTM_User, time=0, value=0}
          {arriveTangent=1.1179, interpMode=RCIM_Cubic, leaveTangent=1.1179, tangentMode=RCTM_User, time=0.2042, value=1}
          {arriveTangent=0.0224, interpMode=RCIM_Cubic, leaveTangent=0.0224, tangentMode=RCTM_User, time=1, value=0}
         ]
        }}})
     Vector2D=&Vector2DFromFloat(Value=180)
    )
   ) {isSetParameters=true}
   /Game/Effects/Niagara/Modules/Position/FN_PositionAlongSplineMesh(
    KillOverflow=false
    Progress:NiagaraFloat=&FloatFromCurve(
     FloatCurve=DataInterface() {propertyValues {Curve {keys [
         {arriveTangent=-0.2977, interpMode=RCIM_Cubic, leaveTangent=-0.2977, tangentMode=RCTM_User, time=0, value=1}
         {arriveTangent=-0.0785, interpMode=RCIM_Cubic, leaveTangent=-0.0785, tangentMode=RCTM_User, time=1.6458, value=0.575}
         {arriveTangent=-0.0245, interpMode=RCIM_Cubic, leaveTangent=-0.0245, tangentMode=RCTM_User, time=2.1383, value=0.5421}
         {arriveTangent=-0.0326, interpMode=RCIM_Cubic, leaveTangent=-0.0326, time=2.6211, value=0.5313}
         {arriveTangent=-0.1695, interpMode=RCIM_Cubic, leaveTangent=-0.1695, tangentMode=RCTM_User, time=2.9273, value=0.5164}
         {arriveTangent=-0.039, interpMode=RCIM_Cubic, leaveTangent=-0.039, tangentMode=RCTM_User, time=5.9845, value=0.045}
         {arriveTangent=0.0173, interpMode=RCIM_Cubic, leaveTangent=0.0173, tangentMode=RCTM_User, time=8, value=0.02}
        ]}}}
     CurveIndex=$Emitter.LoopedAge
    )
    Static Mesh=DataInterface() {propertyValues {
      DefaultMesh=/Game/Tools/SplineMeshGenerator/Saves/Meshes/SM_Emote_GalaxySwish_StarPos
      InstanceIndex=-1
      LODIndex=0
      PreviewMesh=&SM_Emote_GalaxySwish_StarPos
      SoftSourceActor=null
      SourceMode=DefaultMeshOnly
      bAllowSamplingFromStreamingLODs=false
      bCaptureTransformsPerFrame=true
      bUsePhysicsBodyVelocity=false
      FilteredSockets []
      LODIndexUserParameter {parameter {name=None, typeDefHandle {registeredTypeIndex=108}, varData []}}
      MeshParameterBinding {parameter {name=None, typeDefHandle {registeredTypeIndex=1}, varData []}}
      SectionFilter {allowedMaterialSlots []}
     }}
   )
   /SetVariables(Particles.Position=&AddVectorToPosition(
     Position=$Particles.Position
     Vector=&AddVectorToPosition(
      Position=&AddVectorToPosition(
       Position=<-0.5 0 -2>
       Vector=&MakeVector(
        X=0
        Y=0
        Z=&FloatFromCurve(FloatCurve=DataInterface() {propertyValues {Curve {keys [<at 8.7 0 RCIM_Cubic>,<at 11.3 12 RCIM_Cubic>]}}}, CurveIndex=$Emitter.LoopedAge)
       )
      )
      Vector=&MakeVector(X=0, Y=0, Z=&Waveform("Waveform [1]"=::NewEnumerator0, "[1] Frequency"=1, "[1] Amplitude Min/Max"=<0 1>))
     )
    )) {isSetParameters=true}
   /SetVariables(Particles.GlowPos:NiagaraPosition=&AddVectorToPosition(Position=$Particles.Position, Vector=<0 0 4>)) {isSetParameters=true}
   /SetVariables(Particles.FinalPos:NiagaraPosition=&AddVectorToPosition(Position=$Particles.GlowPos, Vector=<0 -4 0>)) {isSetParameters=true}
   &FN_Light_Attributes(
    Color=<rgba 1 0.6601 0.0738 1>
    Create Color Attribute=true
    Create Exponent Attribute=true
    Create Position Attribute=true
    Create Radius Attribute=true
    Exponent=4
    Position=&AddVectorToPosition(Position=$Particles.Position, Vector=<0 5 5>)
    Radius=60
    Scale Color=&Waveform("Waveform [1]"=::NewEnumerator2, "[1] Frequency"=1, "[1] Amplitude Min/Max"=<9 16>)
   )
   &CalculateAccurateVelocity()
  ]
  rendererClasses [
   &NiagaraMeshRendererProperties
   &NiagaraSpriteRendererProperties
   &NiagaraLightRendererProperties
  ]
  renderers [
   {rendererClass=&NiagaraMeshRendererProperties, rendererIndex=0}
   {rendererClass=&NiagaraSpriteRendererProperties, rendererIndex=1}
   {rendererClass=&NiagaraSpriteRendererProperties, rendererIndex=2}
   {rendererClass=&NiagaraLightRendererProperties, rendererIndex=3}
   {rendererClass=&NiagaraSpriteRendererProperties, rendererIndex=4}
   {rendererClass=&NiagaraSpriteRendererProperties, rendererIndex=5}
   {rendererClass=&NiagaraMeshRendererProperties, rendererIndex=6}
   {rendererClass=&NiagaraSpriteRendererProperties, rendererIndex=7}
  ]
 }
 MainStar_Secondary {
  emitterUpdateScript [
   &EmitterState(
    Inactive Response=::NewEnumerator1
    Life Cycle Mode=::NewEnumerator1
    Loop Behavior=::NewEnumerator1
    Loop Delay=1.7
    Loop Duration=13.05
    UseLoopDelay=true
   )
   &SpawnParticlesFromOtherEmitter(
    Attribute Reader=DataInterface() {propertyValues {EmitterBinding {bindingMode=Other, emitterName=MainStar}}}
    Spawn Rate=&FloatFromCurve(
     FloatCurve=DataInterface() {propertyValues {Curve {keys [
         <at 0 1 RCIM_Cubic>
         {arriveTangent=-0.4348, interpMode=RCIM_Cubic, leaveTangent=-0.4348, time=2, value=1}
         <at 2.3 0 RCIM_Cubic>
         {arriveTangent=0.2703, interpMode=RCIM_Cubic, leaveTangent=0.2703, time=2.5, value=1}
         {arriveTangent=-0.2632, interpMode=RCIM_Cubic, leaveTangent=-0.2632, time=6, value=1}
         <at 6.3 0 RCIM_Cubic>
        ]}}}
     CurveIndex=$Emitter.LoopedAge
     Scale Curve=48
    )
    Use Maximum Age=false
    Use Minimum Age=false
   )
  ]
  particleSpawnScript [
   &InitializeParticle(
    Color Mode=::NewEnumerator0
    Lifetime Max=0.9
    Lifetime Min=0.3
    Lifetime Mode=::NewEnumerator1
    Sprite Size Mode=::NewEnumerator1
    Uniform Sprite Size Max=1.25
    Uniform Sprite Size Min=0.45
    UsePositionOffset=false
    Write Material Random=true
   )
   &SampleParticlesFromOtherEmitter(Applied Velocity Scale=-0.3, Velocity Sampling=::NewEnumerator0)
   &ShapeLocation(MaskSpawnGroup=false, Override Seed=false, Sphere Radius=8, UseOwnerScale=true)
  ]
  particleUpdateScript [
   &ParticleState()
   &CurlNoiseForce(Bias Noise=false, Noise Frequency=20, Noise Strength=50, Pan Noise=false)
   &Drag(Drag=&RandomRangeFloat(Minimum=0.8, Maximum=1.2), Use Linear Drag=true, Use Rotational Drag=true)
   &SolveForcesAndVelocity(Clamp Velocity=false, Limit Acceleration=false)
   &ScaleSpriteSizeBySpeed(Max Scale Factor=<0.2 5>, Velocity Threshold=150)
   &ScaleColor(
    Scale Alpha=&FloatFromCurve(FloatCurve=DataInterface() {propertyValues {Curve {
        defaultValue=1
        keys [
         {arriveTangent=9.0725, interpMode=RCIM_Cubic, leaveTangent=9.0725, tangentMode=RCTM_User, time=0, value=0}
         {arriveTangent=-2.5249, interpMode=RCIM_Cubic, leaveTangent=-2.5249, tangentMode=RCTM_User, time=0.1408, value=0.9847}
         {arriveTangent=-0.0405, interpMode=RCIM_Cubic, leaveTangent=-0.0405, tangentMode=RCTM_User, time=1, value=0}
        ]
       }}})
    ScaleA=true
    ScaleRGB=true
   )
  ]
  rendererClasses [&NiagaraSpriteRendererProperties]
  renderers [{rendererClass=&NiagaraSpriteRendererProperties}]
 }
 SmokeRing {
  emitterUpdateScript [
   &EmitterState(Inactive Response=::NewEnumerator1, Life Cycle Mode=::NewEnumerator1, Loop Behavior=::NewEnumerator1, Loop Duration=16, UseLoopDelay=false)
   &SpawnRate(
    SpawnRate=&FloatFromCurve(
     FloatCurve=DataInterface() {propertyValues {Curve {
        defaultValue=1
        keys [
         {arriveTangent=20.6992, interpMode=RCIM_Cubic, leaveTangent=20.6992, tangentMode=RCTM_User, time=0.1963, value=0.3}
         {interpMode=RCIM_Cubic, tangentMode=RCTM_User, time=0.2981, value=0.9882}
         {arriveTangent=0.0671, interpMode=RCIM_Cubic, leaveTangent=0.0671, tangentMode=RCTM_User, time=0.9, value=1}
         {arriveTangent=0.0296, interpMode=RCIM_Cubic, leaveTangent=0.0296, tangentMode=RCTM_User, time=1, value=0}
        ]
       }}}
     CurveIndex=$Emitter.NormalizedLoopAge
     Scale Curve=15
    )
    Use Spawn Probability=false
   )
   &SpawnBurst_Instantaneous(Spawn Count=15, Spawn Group=1, Spawn Time=0.78, Use Spawn Probability=false, UseLoopCountLimit=false)
   &SpawnRate(
    SpawnGroup=1
    SpawnRate=&FloatFromCurve(
     FloatCurve=DataInterface() {propertyValues {Curve {
        defaultValue=1
        keys [
         <at 0.05 0 RCIM_Cubic>
         {interpMode=RCIM_Cubic, tangentMode=RCTM_User, time=0.08, value=1}
         {
          arriveTangent=-0.1498
          arriveTangentWeight=0.4224
          interpMode=RCIM_Cubic
          leaveTangent=-0.1498
          leaveTangentWeight=0.049
          tangentMode=RCTM_User
          time=0.875
          value=1
         }
         {arriveTangent=0.0296, interpMode=RCIM_Cubic, leaveTangent=0.0296, tangentMode=RCTM_User, time=0.95, value=0}
        ]
       }}}
     CurveIndex=$Emitter.NormalizedLoopAge
     Scale Curve=18
    )
    Use Spawn Probability=false
   ) {module=SpawnRate001}
   &SpawnBurst_Instantaneous(Spawn Group=2, Use Spawn Probability=false, UseLoopCountLimit=false) {module=SpawnBurst_Instantaneous001}
   &SpawnBurst_Instantaneous(Spawn Group=3, Use Spawn Probability=false, UseLoopCountLimit=false) {module=SpawnBurst_Instantaneous002}
   &SpawnBurst_Instantaneous(Spawn Group=4, Use Spawn Probability=false, UseLoopCountLimit=false) {module=SpawnBurst_Instantaneous003}
   &SpawnBurst_Instantaneous(Spawn Group=5, Use Spawn Probability=false, UseLoopCountLimit=false) {module=SpawnBurst_Instantaneous005}
   &SpawnBurst_Instantaneous(Spawn Group=6, Spawn Time=14.7, Use Spawn Probability=false, UseLoopCountLimit=false) {module=SpawnBurst_Instantaneous006}
   &SpawnBurst_Instantaneous(Spawn Group=7, Spawn Time=14.7, Use Spawn Probability=false, UseLoopCountLimit=false) {module=SpawnBurst_Instantaneous007}
   &SpawnBurst_Instantaneous(Spawn Group=8, Spawn Time=14.7, Use Spawn Probability=false, UseLoopCountLimit=false) {module=SpawnBurst_Instantaneous008}
   &SpawnBurst_Instantaneous(Spawn Group=5, Use Spawn Probability=false, UseLoopCountLimit=false) {module=SpawnBurst_Instantaneous004}
   /SetVariables(Emitter.0 - SpeedDir=-0.3, Emitter.1 - SpeedDir=-0.7) {isSetParameters=true}
  ]
  particleSpawnScript [
   /SetVariables(Particles.VisibilityTag=&SpawnGroup()) {isSetParameters=true}
   &InitializeParticle(
    Color=<rgba 3 1.9064 0.6691 0.2>
    Lifetime=&MergeHelper_FloatSwitch(
     0=&RandomRangeFloat(Minimum=0.5, Maximum=0.9)
     1=&RandomRangeFloat(Minimum=0.4, Maximum=0.9)
     2=16
     3=16
     4=16
     5=16
     6=0.48
     7=0.3
     8=0.3
     Index=&SpawnGroup()
     Num=::NewEnumerator7
    )
    Mesh Scale=&Multiply_VectorByFloat(
     Vector=&MergeHelper_VectorSwitch(
      0=&RandomRangeVector(Minimum=<1.7 1.7 1.7>, Maximum=<1.75 1.75 2.1>)
      1=&RandomRangeVector(Minimum=<1.6 1.6 1.6>, Maximum=<1.65 1.65 1.8>)
      2=<0 0 0>
      3=<0 0 0>
      4=<3 3 3>
      5=&VectorFromFloat(Value=1.7005)
      6=&VectorFromFloat(Value=2.1)
      7=&VectorFromFloat(Value=2.8)
      8=&VectorFromFloat(Value=3.1736)
      Index=&SpawnGroup()
      Num=::NewEnumerator7
     )
     Float=0.8
    )
    Mesh Scale Mode=::NewEnumerator4
    Position Offset=<0 -45 80>
    Sprite Size Mode=::NewEnumerator3
    Uniform Sprite Size=84.3967
    UsePositionOffset=true
    Write Material Random=true
   )
   &ShapeLocation(
    Cylinder Height=&MergeHelper_FloatSwitch(0=0, 1=2, 2=0, 3=0, 4=0, 5=0, Index=&SpawnGroup(), Num=::NewEnumerator4)
    Cylinder Height Midpoint=1
    Cylinder Radius=&MergeHelper_FloatSwitch(0=0, 1=5, 2=0, 3=0, 4=0, 5=0, Index=&SpawnGroup(), Num=::NewEnumerator4)
    MaskSpawnGroup=false
    Offset=&MaskVectorBySpawnGroup(Masked Vector=<0 0 0>, Passthrough Vector=<0 0 5>)
    Offset Mode=::NewEnumerator0
    Override Seed=false
    Shape Primitive=::NewEnumerator1
    Surface Only=false
    UseOwnerScale=true
   )
   &InitialMeshOrientation(
    Mesh Axis To Orient=<0 1 0>
    Mesh Orientation Mode=::NewEnumerator2
    Orientation Vector=<0 0 1>
    Rotation=&RandomRangeVector(Minimum=<0 0 0>, Maximum=<0 0 2>)
    Use Rotation Vector=true
   )
   &SpriteFacingAndAlignment(Sprite Alignment=<-1 0 0>, Sprite Facing=<0 1 0>, WriteAlignment=false, WriteFacing=true)
  ]
  particleUpdateScript [
   &ParticleState()
   &ScaleSpriteSize(
    Uniform Curve Index=$Emitter.NormalizedLoopAge
    Uniform Curve Sprite Scale=DataInterface() {propertyValues {Curve {keys [
        {arriveTangent=21.4695, interpMode=RCIM_Cubic, leaveTangent=21.4695, tangentMode=RCTM_User, time=0, value=0.2}
        {arriveTangent=1, time=0.075, value=1}
        {arriveTangent=0.0865, interpMode=RCIM_Cubic, leaveTangent=0.0865, tangentMode=RCTM_User, time=0.925, value=1}
        {arriveTangent=-17.1415, interpMode=RCIM_Cubic, leaveTangent=-17.1415, tangentMode=RCTM_User, time=1, value=0}
       ]}}}
   )
   &ScaleMeshSize(Scale Factor=&Multiply_VectorByFloat(
     Vector=&MergeHelper_VectorSwitch(
      0=&Multiply_Vector(
       A=&VectorFromFloat(Value=&FloatFromCurve(
         FloatCurve=DataInterface() {propertyValues {Curve {
            defaultValue=1
            keys [
             {interpMode=RCIM_Cubic, tangentMode=RCTM_User, time=0, value=0.6}
             {arriveTangent=0.2858, interpMode=RCIM_Cubic, leaveTangent=0.2858, tangentMode=RCTM_User, time=1, value=1.1}
             <at 1.1 1.1 RCIM_Cubic>
             {arriveTangent=0.0044, interpMode=RCIM_Cubic, leaveTangent=0.0044, tangentMode=RCTM_User, time=15, value=1.1}
             <at 16 0 RCIM_Cubic>
            ]
           }}}
         CurveIndex=$Emitter.LoopedAge
        ))
       B=&VectorFromFloat(Value=&FloatFromCurve(
         FloatCurve=DataInterface() {propertyValues {Curve {
            defaultValue=1
            keys [
             {interpMode=RCIM_Cubic, tangentMode=RCTM_User, time=0, value=1}
             {arriveTangent=0.3803, interpMode=RCIM_Cubic, leaveTangent=0.3803, tangentMode=RCTM_User, time=1, value=1.1}
            ]
           }}}
         Scale Curve=&RandomRangeFloat(Minimum=1, Maximum=1.2)
        ))
      )
      1=&VectorFromFloat(Value=&FloatFromCurve(
        FloatCurve=DataInterface() {propertyValues {Curve {
           defaultValue=1
           keys [<at -0.0078 1 RCIM_Cubic>, {arriveTangent=-0.3116, interpMode=RCIM_Cubic, leaveTangent=-0.3116, tangentMode=RCTM_User, time=1, value=1.5}]
          }}}
        Scale Curve=&RandomRangeFloat(Minimum=1, Maximum=1.25)
       ))
      2=<0 0 0>
      3=<0 0 0>
      4=<1 1 1>
      5=<1 1 1>
      6=&VectorFromFloat(Value=&FloatFromCurve(
        FloatCurve=DataInterface() {propertyValues {Curve {
           defaultValue=1
           keys [
            {arriveTangent=1.724, interpMode=RCIM_Cubic, leaveTangent=1.724, tangentMode=RCTM_User, time=0, value=0}
            {interpMode=RCIM_Cubic, tangentMode=RCTM_User, time=1, value=1}
           ]
          }}}
        Scale Curve=1.2
       ))
      7=&VectorFromFloat(Value=&FloatFromCurve(FloatCurve=DataInterface() {propertyValues {Curve {keys [<at 0 0 RCIM_Cubic>,<at 1 1 RCIM_Cubic>]}}}))
      Index=&SpawnGroup()
      Num=::NewEnumerator6
     )
     Float=&FloatFromCurve(
      FloatCurve=DataInterface() {propertyValues {Curve {keys [
          {arriveTangent=21.4695, interpMode=RCIM_Cubic, leaveTangent=21.4695, tangentMode=RCTM_User, time=0, value=0.2}
          {arriveTangent=1, time=0.075, value=1}
          {arriveTangent=-0.0732, interpMode=RCIM_Cubic, leaveTangent=-0.0732, tangentMode=RCTM_User, time=0.8, value=1}
          {arriveTangent=-0.086, interpMode=RCIM_Cubic, leaveTangent=-0.086, tangentMode=RCTM_User, time=1, value=0}
         ]}}}
      CurveIndex=$Emitter.NormalizedLoopAge
     )
    ))
   &UpdateMeshOrientation(
    Rotation Rate=&Multiply_Float(
     A=&Multiply_Float(
      A=&MergeHelper_FloatSwitch(
       0=&RandomRangeFloat(Minimum=-0.05, Maximum=0.05)
       1=&RandomRangeFloat(Minimum=-0.4, Maximum=0.4)
       2=0
       3=0
       4=0
       5=0
       Index=&SpawnGroup()
       Num=::NewEnumerator4
      )
      B=&FloatFromCurve(FloatCurve=DataInterface() {propertyValues {Curve {keys [
           {interpMode=RCIM_Cubic, tangentMode=RCTM_User, time=0, value=1}
           {arriveTangent=-3.5121, interpMode=RCIM_Cubic, leaveTangent=-3.5121, tangentMode=RCTM_User, time=1.0041, value=0}
          ]}}})
     )
     B=&FloatFromCurve(
      FloatCurve=DataInterface() {propertyValues {Curve {keys [
          {interpMode=RCIM_Cubic, tangentMode=RCTM_User, time=0, value=1}
          {interpMode=RCIM_Cubic, tangentMode=RCTM_User, time=0.0458, value=1.2678}
          {arriveTangent=-1.2911, interpMode=RCIM_Cubic, leaveTangent=-1.2911, tangentMode=RCTM_User, time=0.1562, value=1}
          {
           arriveTangent=-0.3731
           arriveTangentWeight=0.0848
           interpMode=RCIM_Cubic
           leaveTangent=-0.3731
           leaveTangentWeight=0.2422
           tangentMode=RCTM_User
           time=0.4828
           value=0.6758
          }
          {arriveTangent=-0.0256, interpMode=RCIM_Cubic, leaveTangent=-0.0256, tangentMode=RCTM_User, time=0.9951, value=0.5}
         ]}}}
      CurveIndex=$Emitter.NormalizedLoopAge
     )
    )
    Rotation Vector=<0 0 1>
   )
   &PointForce(
    Force Origin=&AddVectorToPosition(
     Position=/Niagara/DynamicInputs/Helpers/SimulationPosition()
     Vector=&MaskVectorBySpawnGroup(Masked Vector=<0 0 0>, Passthrough Vector=<0 0 50>)
    )
    Force Strength=&RandomRangeFloat(
     Minimum=0
     Maximum=&MergeHelper_FloatSwitch(0=15, 1=30, 2=0, 3=0, 4=0, 5=0, Index=&SpawnGroup(), Num=::NewEnumerator4)
    )
    Module.Use Falloff=false
    Module.UseLinearForce=false
   )
   &SolveForcesAndVelocity(Clamp Velocity=false, Limit Acceleration=false)
   &ScaleColor(
    Scale Alpha=&MaskFloatBySpawnGroup(
     Masked Float=&Multiply_Float(
      A=&FloatFromCurve(
       FloatCurve=DataInterface() {propertyValues {Curve {
          defaultValue=1
          keys [
           <at 0 0 RCIM_Cubic>
           {arriveTangent=0.0037, interpMode=RCIM_Cubic, leaveTangent=0.0037, tangentMode=RCTM_User, time=1, value=1}
           {arriveTangent=0.0007, interpMode=RCIM_Cubic, leaveTangent=0.0007, tangentMode=RCTM_User, time=14, value=1}
           {arriveTangent=-0.0019, interpMode=RCIM_Cubic, leaveTangent=-0.0019, tangentMode=RCTM_User, time=15, value=0}
           <at 16 0 RCIM_Cubic>
          ]
         }}}
       CurveIndex=$Emitter.LoopedAge
      )
      B=&FloatFromCurve(FloatCurve=DataInterface() {propertyValues {Curve {
          defaultValue=1
          keys [
           {interpMode=RCIM_Cubic, tangentMode=RCTM_User, time=0, value=0}
           {interpMode=RCIM_Cubic, tangentMode=RCTM_User, time=0.5, value=1}
           {tangentMode=RCTM_User, time=1, value=0}
          ]
         }}})
     )
     Passthrough Float=&Multiply_Float(
      A=&FloatFromCurve(
       FloatCurve=DataInterface() {propertyValues {Curve {
          defaultValue=1
          keys [
           {interpMode=RCIM_Cubic, tangentMode=RCTM_User, time=0, value=0}
           {interpMode=RCIM_Cubic, tangentMode=RCTM_User, time=0.1617, value=0.6549}
           {tangentMode=RCTM_User, time=1, value=0}
          ]
         }}}
       Scale Curve=0.6
      )
      B=&FloatFromCurve(
       FloatCurve=DataInterface() {propertyValues {Curve {
          defaultValue=1
          keys [
           <at 0.6866 1 RCIM_Cubic>
           {arriveTangent=-7.3431, interpMode=RCIM_Cubic, leaveTangent=-7.3431, tangentMode=RCTM_User, time=0.7817, value=0.2843}
           {arriveTangent=-0.6899, interpMode=RCIM_Cubic, leaveTangent=-0.6899, tangentMode=RCTM_User, time=0.85, value=0}
          ]
         }}}
       CurveIndex=$Emitter.NormalizedLoopAge
      )
     )
    )
    Scale RGB=&MaskVectorBySpawnGroup(
     Masked Vector=<1 1 1>
     Passthrough Vector=&MakeVectorFromLinearColorRGB(Color=/Niagara/DynamicInputs/Add/Add_LinearColor(
       A:LinearColor=&ColorFromCurve(ColorCurve=DataInterface() {propertyValues {
          AlphaCurve {keys [
            <at 0 0>
            {arriveTangent=1.2772, interpMode=RCIM_Cubic, leaveTangent=1.2772, time=0.0901, value=1}
            <at 0.3915 0.5>
            <at 0.875 0.193>
            <at 0.9 0 RCIM_Cubic>
           ]}
          BlueCurve {keys [<at 0.009 0.2782>,<at 0.0499 0.2461>,<at 0.2996 0.364>,<at 1 0.055 RCIM_Cubic>]}
          GreenCurve {keys [<at 0.009 0.8466>,<at 0.0499 0.6916>,<at 0.2996 0.7529>,<at 1 0.1592 RCIM_Cubic>]}
          RedCurve {keys [<at 0.009 1>,<at 0.0499 0.9927>,<at 0.2996 1>,<at 1 0.2031 RCIM_Cubic>]}
         }})
       B:LinearColor=/Niagara/DynamicInputs/LinearColor/MakeLinearColorFromHSV(
        HSV From:NiagaraInt32=1
        HSV:Vector3f=?
        Hue:NiagaraFloat=0
        Saturation:NiagaraFloat=0
        Value:NiagaraFloat=&FloatFromCurve(
         FloatCurve=DataInterface() {propertyValues {Curve {keys [
             {arriveTangent=60.3451, interpMode=RCIM_Cubic, leaveTangent=60.3451, tangentMode=RCTM_User, time=-0.0007, value=1}
             {arriveTangent=-14.8494, interpMode=RCIM_Cubic, leaveTangent=-14.8494, tangentMode=RCTM_User, time=0.0405, value=0.3673}
             {interpMode=RCIM_Cubic, tangentMode=RCTM_User, time=0.1045, value=0}
            ]}}}
         CurveIndex=$Emitter.NormalizedLoopAge
        )
        Alpha:NiagaraFloat=0
       )
      ))
    )
    ScaleA=true
    ScaleRGB=true
   )
   /SetVariables(
    Particles.2DP:Vector4f=&MakeVector4(
     X=&FloatFromCurve(FloatCurve=DataInterface() {propertyValues {Curve {
         defaultValue=1
         keys [
          <at 0 0 RCIM_Cubic>
          {arriveTangent=0.1839, interpMode=RCIM_Cubic, leaveTangent=0.1839, tangentMode=RCTM_User, time=0.0931, value=0.9}
          {interpMode=RCIM_Cubic, tangentMode=RCTM_User, time=0.6, value=0.9}
          {arriveTangent=-5.706, interpMode=RCIM_Cubic, leaveTangent=-5.706, tangentMode=RCTM_User, time=1, value=0}
         ]
        }}})
     Y=0
     Z=0
     W=0
    )
    Particles.2_1Size:Vector2f=&Multiply_Vector2D_ByFloat(Float=1.9, Vector2D=$Particles.SpriteSize)
   ) {isSetParameters=true}
   /SetVariables(
    Particles.3Color:LinearColor=&ColorFromCurve(ColorCurve=DataInterface() {propertyValues {
       AlphaCurve {keys [<at 0 0 RCIM_Cubic>,<at 0.0603 0.05>,<at 0.65 0.05>,<at 0.665 0.125>,<at 0.7 0.075>,<at 0.875 0.075>,<at 0.9 0 RCIM_Cubic>]}
       BlueCurve {keys [<at 0 0.7607 RCIM_Cubic>]}
       GreenCurve {keys [<at 0 4.2439 RCIM_Cubic>]}
       RedCurve {keys [<at 0 6 RCIM_Cubic>]}
      }})
    Particles.3Size:Vector2f=&Multiply_Vector2D_ByFloat(Float=2.45, Vector2D=$Particles.SpriteSize)
   ) {isSetParameters=true}
   /SetVariables(
    Particles.4Color:LinearColor=&ColorFromCurve(ColorCurve=DataInterface() {propertyValues {
       AlphaCurve {keys [<at 0 0>,<at 0.06 0.05>,<at 0.065 0.5>,<at 0.07 0.05>,<at 0.65 0.15>,<at 0.665 1>,<at 0.7 0.4>,<at 0.875 0.4>,<at 0.9 0 RCIM_Cubic>]}
       BlueCurve {keys [<at 0.85 0.2225 RCIM_Cubic>,<at 0.9977 0>]}
       GreenCurve {keys [<at 0.85 1.2014 RCIM_Cubic>,<at 0.9977 0.2132>]}
       RedCurve {keys [<at 0.85 3 RCIM_Cubic>,<at 0.9977 0.4375>]}
      }})
    Particles.4DP:Vector4f=&MakeVector4(X=0, Y=1, Z=0.8769, W=0)
   ) {isSetParameters=true}
   /SetVariables(Particles.5Color:LinearColor=&ColorFromCurve(ColorCurve=DataInterface() {propertyValues {
       AlphaCurve {keys [<at 0 0>,<at 0.06 0.05>,<at 0.065 0.5>,<at 0.07 0.05>,<at 0.65 0.15>,<at 0.665 0.6>,<at 0.7 0.225>,<at 0.875 0.225>,<at 0.9 0 RCIM_Cubic>]}
       BlueCurve {keys [<at 0.85 0.2225 RCIM_Cubic>,<at 0.9977 0>]}
       GreenCurve {keys [<at 0.85 1.2014 RCIM_Cubic>,<at 0.9977 0.2132>]}
       RedCurve {keys [<at 0.85 3 RCIM_Cubic>,<at 0.9977 0.4375>]}
      }})) {isSetParameters=true}
   /SetVariables(
    Particles.6Color:LinearColor=&MakeLinearColorFromVectorAndFloat(
     "Vector (RGB)"=&MakeVectorFromLinearColorRGB(Color=<rgba 3 1.6234 0.3687 1>)
     "Float (Alpha)"=&FloatFromCurve(FloatCurve=DataInterface() {propertyValues {Curve {
         defaultValue=1
         keys [
          {arriveTangent=9.0725, interpMode=RCIM_Cubic, leaveTangent=9.0725, tangentMode=RCTM_User, time=0, value=0}
          {arriveTangent=-0.0481, interpMode=RCIM_Cubic, leaveTangent=-0.0481, tangentMode=RCTM_User, time=0.1408, value=0.9847}
          {arriveTangent=-0.0405, interpMode=RCIM_Cubic, leaveTangent=-0.0405, tangentMode=RCTM_User, time=1, value=0}
         ]
        }}})
    )
    Particles.6DP:Vector4f=&MakeVector4(
     X=&FloatFromCurve(FloatCurve=DataInterface() {propertyValues {Curve {keys [{arriveTangent=-2.1015, interpMode=RCIM_Cubic, leaveTangent=-2.1015, tangentMode=RCTM_User, time=0, value=0.5}, <at 1 -0.15>]}}})
     Y=&FloatFromCurve(FloatCurve=DataInterface() {propertyValues {Curve {keys [{arriveTangent=-0.3464, interpMode=RCIM_Cubic, leaveTangent=-0.3464, tangentMode=RCTM_User, time=0, value=1}, <at 1 0.9>]}}})
     Z=&FloatFromCurve(FloatCurve=DataInterface() {propertyValues {Curve {keys [<at 0 0>,<at 0.15 1>,<at 1 1.2>]}}})
     W=0
    )
    Particles.6Scale:Vector3f=&VectorFromFloat(Value=2.1)
   ) {isSetParameters=true}
   /SetVariables(
    Particles.7Color:LinearColor=&MakeLinearColorFromVectorAndFloat(
     "Vector (RGB)"=&MakeVectorFromLinearColorRGB(Color=<rgba 12 6.4937 1.4748 1>)
     "Float (Alpha)"=&FloatFromCurve(
      FloatCurve=DataInterface() {propertyValues {Curve {
         defaultValue=1
         keys [
          {arriveTangent=9.0725, interpMode=RCIM_Cubic, leaveTangent=9.0725, tangentMode=RCTM_User, time=0, value=0}
          {arriveTangent=-0.0481, interpMode=RCIM_Cubic, leaveTangent=-0.0481, tangentMode=RCTM_User, time=0.1408, value=0.9847}
          {arriveTangent=-0.0405, interpMode=RCIM_Cubic, leaveTangent=-0.0405, tangentMode=RCTM_User, time=1, value=0}
         ]
        }}}
      Scale Curve=0.275
     )
    )
    Particles.7Size:Vector2f=&Multiply_Vector2D_ByFloat(
     Float=&FloatFromCurve(FloatCurve=DataInterface() {propertyValues {Curve {
         defaultValue=0.2
         keys [
          {arriveTangent=4.1785, interpMode=RCIM_Cubic, leaveTangent=4.1785, tangentMode=RCTM_User, time=0, value=0}
          {arriveTangent=1.1179, interpMode=RCIM_Cubic, leaveTangent=1.1179, tangentMode=RCTM_User, time=0.2042, value=0.6228}
          {arriveTangent=0.0224, interpMode=RCIM_Cubic, leaveTangent=0.0224, tangentMode=RCTM_User, time=1, value=1}
         ]
        }}})
     Vector2D=&Vector2DFromFloat(Value=250)
    )
   ) {isSetParameters=true}
   &FN_Light_Attributes(
    Color=<rgba 1 0.5564 0.125 1>
    Create Color Attribute=true
    Create Exponent Attribute=true
    Create Position Attribute=true
    Create Radius Attribute=true
    Exponent=2
    Position=<0 -32.6601 30>
    Radius=80
    Scale Color=&FloatFromCurve(
     FloatCurve=DataInterface() {propertyValues {Curve {keys [<at 0 0>,<at 0.075 1>,<at 0.9 1>,<at 1 0>]}}}
     Scale Curve=&Waveform("Waveform [1]"=::NewEnumerator2, "[1] Frequency"=1, "[1] Amplitude Min/Max"=<18 24>)
    )
   )
  ]
  rendererClasses [
   &NiagaraMeshRendererProperties
   &NiagaraSpriteRendererProperties
   &NiagaraLightRendererProperties
  ]
  renderers [
   {rendererClass=&NiagaraMeshRendererProperties, rendererIndex=0}
   {rendererClass=&NiagaraMeshRendererProperties, rendererIndex=1}
   {rendererClass=&NiagaraSpriteRendererProperties, rendererIndex=2}
   {rendererClass=&NiagaraSpriteRendererProperties, rendererIndex=3}
   {rendererClass=&NiagaraMeshRendererProperties, rendererIndex=4}
   {rendererClass=&NiagaraLightRendererProperties, rendererIndex=5}
   {rendererClass=&NiagaraMeshRendererProperties, rendererIndex=6}
   {rendererClass=&NiagaraMeshRendererProperties, rendererIndex=7}
   {rendererClass=&NiagaraMeshRendererProperties, rendererIndex=8}
   {rendererClass=&NiagaraSpriteRendererProperties, rendererIndex=9}
   {rendererClass=&NiagaraSpriteRendererProperties, rendererIndex=10}
  ]
 }
}
summary {
 systemName=NS_Emote_GalaxySwish_BG
 emitters [
  {
   emitterName=SmokeRing
   rendererClasses [
    &NiagaraMeshRendererProperties
    &NiagaraSpriteRendererProperties
    &NiagaraLightRendererProperties
   ]
  }
  {emitterName=FallingStars, rendererClasses [&NiagaraSpriteRendererProperties]}
  {emitterName=FallingStars_Secondary, rendererClasses [&NiagaraSpriteRendererProperties]}
  {emitterName=ConvergingStars, rendererClasses [&NiagaraSpriteRendererProperties]}
  {emitterName=Constellation, rendererClasses [&NiagaraMeshRendererProperties]}
  {
   emitterName=MainStar
   rendererClasses [
    &NiagaraMeshRendererProperties
    &NiagaraSpriteRendererProperties
    &NiagaraLightRendererProperties
   ]
  }
  {emitterName=MainStar_Secondary, rendererClasses [&NiagaraSpriteRendererProperties]}
 ]
 userVariables [
  User.bIsFrontEnd {type=NiagaraBool}
  User.bIsFrontEndPreview {type=NiagaraBool}
 ]
}
