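# AnimBP for the GalaxySwish quadruped cosmetic companion that retargets its pose from a mesh and drives an emote background effect.
# Reverse retarget selects the emote retargeter; when it ends, the ears retargeter holds until a delayed EarsRetarget flag blends to the base retargeter, layered with an accessory head/neck channel blend at zero alpha.
# The NS_Active curve spawns or deactivates the attached Niagara system.
# Init sets ReduceHeadAlpha from the Stage1 outfit variant, though nothing here reads it.
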AnimBlueprint /CosmeticCompanions/Assets/Quadruped/GalaxySwish/Meshes/GalaxySwish_Quadruped_AnimBP
target_skeleton=&/GalaxySwish_Quadruped_Skeleton
graphs [
 AnimGraph {nodes [
   n1 = Blend Bone By Channel!(
    A:Pose Link Structure=Retarget Pose From Mesh!() -> (Pose:Pose Link Structure) {is=title AnimGraphNode_RetargetPoseFromMesh, iKRetargeterAsset=&/RTG_GalaxySwish}
    B:Pose Link Structure=Retarget Pose From Mesh() {iKRetargeterAsset=&/RTG_GalaxySwish_accessory}
    Alpha:float=0
   ) -> (Pose:Pose Link Structure) {
    is=title AnimGraphNode_BlendBoneByChannel
    blendNode {boneDefinitions [
      {bBlendRotation=true, bBlendScale=true, bBlendTranslation=true, sourceBone {boneName=head}, targetBone {boneName=head}}
      {bBlendRotation=true, bBlendScale=true, bBlendTranslation=true, sourceBone {boneName=neck_02}, targetBone {boneName=neck_02}}
      {bBlendRotation=true, bBlendScale=true, bBlendTranslation=true, sourceBone {boneName=neck_01}, targetBone {boneName=neck_01}}
      {bBlendRotation=true, bBlendScale=true, bBlendTranslation=true, sourceBone {boneName=attach_emote_01}, targetBone {boneName=attach_emote_01}}
      {bBlendRotation=true, bBlendScale=true, bBlendTranslation=true, sourceBone {boneName=attach_emote_02}, targetBone {boneName=attach_emote_02}}
     ]}
   }
   n2 = Retarget Pose From Mesh() {iKRetargeterAsset=&/RTG_GalaxySwish_Emote}
   n3 = Retarget Pose From Mesh() {iKRetargeterAsset=&/RTG_GalaxySwish_ears}
   n4 = Get bUseReverseRetarget() -> (bUseReverseRetarget:bool)
   Output Pose(Result=Blend Poses by bool(
     BlendPose_0=@n1
     BlendPose_1=Blend Poses by bool(BlendPose_0=@n2, BlendPose_1=@n3, BlendTime_1=0.3, bActiveValue=@n4)
     BlendTime_0=0.3
     BlendTime_1=0.3
     bActiveValue=Get EarsRetarget() -> (EarsRetarget:bool)
    ))
  ]}
 EventGraph {nodes [
   Event BlueprintUpdateAnimation()
   > n1 = Sequence!(execute) -> (then_0, then_1, then_2:Exec)
   > Niagara()
   n2 = Branch(execute=@n1.then_1, Condition=Get bUseReverseRetarget())
   > Set EarsRetarget(EarsRetarget:bool=false) -> (Output_Get:bool)
   n3 = Branch(execute=@n2.else, Condition=Get EarsRetarget())
   Delay(execute=@n3.else, Duration=1)
   > Set EarsRetarget(EarsRetarget=true)
   Event BlueprintInitializeAnimation()
   > n4 = Cast To CompanionPostProcessAnimInstance!(
    execute
    Object=GetPostProcessInstance!(self:Skeletal Mesh Component Object Reference=GetOwningComponent()) -> (ReturnValue:Anim Instance Object Reference)
   ) -> (then, CastFailed, AsCompanion Post Process Anim Instance:Companion Post Process Anim Instance Object Reference) {is=title K2Node_DynamicCast}
   > n5 = Branch(Condition=HasVariantTagForChannel!(
     self:Fort Player Companion Anim Instance Object Reference=GetMainAnimInstance!(self:Companion Post Process Anim Instance Object Reference=@n4) -> (ReturnValue:Fort Player Companion Anim Instance Object Reference)
     ChannelTag:Gameplay Tag Structure=(TagName="Cosmetics.Variant.Channel.Outfit")
     VariantTag:Gameplay Tag Structure=(TagName="Cosmetics.Variant.Property.Stage1")
    ) -> (ReturnValue:bool))
   > Set ReduceHeadAlpha(ReduceHeadAlpha:double=0.3) -> (Output_Get:double)
   Set ReduceHeadAlpha(execute=@n5.else, ReduceHeadAlpha=0)
  ]}
 Niagara {nodes [
   Niagara() {is=title K2Node_FunctionEntry}
   > n1 = Branch(Condition="float > float"(A=GetCurveValue(CurveName=NS_Active)))
   > n2 = Branch(Condition=Get bNS_Spawned() -> (bNS_Spawned:bool))
   n3 = SpawnSystemAttached!(
    execute=@n2.else
    SystemTemplate:Niagara System Object Reference=/CosmeticCompanions/FX/Builtins/GalaxySwish/NS_Emote_GalaxySwish_BG
    AttachToComponent:Scene Component Object Reference=GetOwningComponent()
    AttachPointName:Name=root
    Location:Vector="0,-45,0"
    Rotation:Rotator="0, 0, 0"
    LocationType:EAttachLocation Enum=SnapToTarget
    bAutoDestroy:bool=true
    bAutoActivate:bool=true
    PoolingMethod:ENCPoolMethod Enum=None
    bPreCullCheck:bool=true
   ) -> (then, ReturnValue:Niagara Particle System Component Object Reference)
   > Set SpawnedNS(SpawnedNS:Niagara Particle System Component Object Reference=@n3) -> (Output_Get:Niagara Particle System Component Object Reference)
   > Set bNS_Spawned(bNS_Spawned:bool=true) -> (Output_Get:bool)
   Branch(execute=@n1.else, Condition=Get bNS_Spawned())
   > Deactivate(self:Actor Component Object Reference=Get SpawnedNS() -> (SpawnedNS:Niagara Particle System Component Object Reference))
   > Set bNS_Spawned(bNS_Spawned=false)
  ]}
]
variables [
 bNS_Spawned {type=boolean}
 EarsRetarget {type=boolean}
 ReduceHeadAlpha {type=number}
 SpawnedNS {type=NiagaraComponent, default=null}
]
