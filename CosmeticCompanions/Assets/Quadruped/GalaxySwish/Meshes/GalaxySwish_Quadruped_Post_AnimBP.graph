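# Post-process AnimBlueprint for the GalaxySwish quadruped cosmetic companion, correcting extra front-limb bones and the chest.
# It applies a PoseAsset to the input, replaces extra wrist/elbow rotations, constrains Chest translation to neck_01, then adds a chest offset and scale.
# Post-evaluation, the owner's skeletal mesh neck_01 yaw sets that chest alpha.
# It also derives wrist/elbow rotators from front-limb deltas, but the AnimGraph shows fixed rotations instead.
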AnimBlueprint /CosmeticCompanions/Assets/Quadruped/GalaxySwish/Meshes/GalaxySwish_Quadruped_Post_AnimBP
target_skeleton=&/GalaxySwish_Quadruped_Skeleton
graphs [
 AnimGraph {nodes [
   n1 = Input Pose()
   n2 = Get PoseAsset() -> (PoseAsset:Pose Asset Object Reference)
   n3 = "Transform (Modify) Bone - Bone: extra_wrist_fr_l"(
    ComponentPose=Local To Component(LocalPose=Pose!(PoseAsset:Pose Asset Object Reference=@n2, SourcePose:Pose Link Structure=@n1) -> (Pose:Pose Link Structure) {is=title AnimGraphNode_PoseBlendNode, bIgnorePoseAssetSourceAnimationCurves=true, blendOption=Linear})
    Alpha=1
    Translation="0,0,0"
    Rotation="-0.023,-10.303,-0.307"
    Scale="1,1,1"
   ) {rotationMode=BMM_Replace, rotationSpace=BCS_BoneSpace, boneToModify {boneName=extra_wrist_fr_l}, rotation=<rot -0.023 -10.303 -0.307>}
   n4 = "Transform (Modify) Bone - Bone: extra_elbow_fr_l"(ComponentPose=@n3, Alpha=1, Translation="0,0,0", Rotation="-0.023,-10.303,-0.307", Scale="1,1,1") {rotationMode=BMM_Replace, rotationSpace=BCS_BoneSpace, boneToModify {boneName=extra_elbow_fr_l}, rotation=<rot -0.023 -10.303 -0.307>}
   n5 = "Transform (Modify) Bone - Bone: extra_wrist_fr_r"(ComponentPose=@n4, Alpha=1, Translation="0,0,0", Rotation="-0.023,-10.303,-0.307", Scale="1,1,1") {rotationMode=BMM_Replace, rotationSpace=BCS_BoneSpace, boneToModify {boneName=extra_wrist_fr_r}, rotation=<rot -0.023 -10.303 -0.307>}
   n6 = "Transform (Modify) Bone - Bone: extra_elbow_fr_r"(ComponentPose=@n5, Alpha=1, Translation="0,0,0", Rotation="-0.023,-10.303,-0.307", Scale="1,1,1") {rotationMode=BMM_Replace, rotationSpace=BCS_BoneSpace, boneToModify {boneName=extra_elbow_fr_r}, rotation=<rot -0.023 -10.303 -0.307>}
   n7 = "Constraint - Chest (T)"(ComponentPose=@n6, Alpha=1, ConstraintWeights_0=1) {boneToModify {boneName=Chest}, constraintSetup [{transformType=Translation, targetBone {boneName=neck_01}}], constraintWeights [1]}
   n8 = "Transform (Modify) Bone - Bone: Chest"(ComponentPose=@n7, Alpha=Get Chest_Alpha() -> (Chest_Alpha:double), Translation="1.751,6.240,0.000", Rotation="0,0,0", Scale="1,1,4") {
    rotationMode=BMM_Additive
    rotationSpace=BCS_ParentBoneSpace
    scaleMode=BMM_Additive
    scaleSpace=BCS_ParentBoneSpace
    translationMode=BMM_Additive
    translationSpace=BCS_ParentBoneSpace
    boneToModify {boneName=Chest}
    scale {z=4}
    translation=<1.751 6.24>
   }
   Output Pose(Result=Component To Local(ComponentPose=@n8))
  ]}
 EventGraph {nodes [
   Event BlueprintPostEvaluateAnimation() {is=title K2Node_Event}
   n1 = GetComponentByClass!(
    self:Actor Object Reference=GetOwningActor(self=Self-Reference())
    ComponentClass:Actor Component Class Reference=/Script/Engine.SkeletalMeshComponent
   ) -> (ReturnValue:Skeletal Mesh Component Object Reference)
   > Set Chest_Alpha(Chest_Alpha:double=Remap Float(
     Input=GetSocketTransform!(self=@n1, InSocketName=neck_01, TransformSpace=RTS_ParentBoneSpace) -> (
      ReturnValue_Location:Vector
      ReturnValue_Rotation_Roll:float
      ReturnValue_Rotation_Pitch:float
      ReturnValue_Rotation_Yaw:float
      ReturnValue_Scale:Vector
     ).ReturnValue_Rotation_Yaw
     OldRange=(X=-40,Y=0)
     NewRange=(X=0,Y=1)
     Clamp=true
    ) -> (Output=~)) -> (Output_Get:double)
   n2 = "Delta (Rotator)"!(
    A:Rotator=GetSocketTransform!(self=@n1, InSocketName=ball_fr_l, TransformSpace=RTS_ParentBoneSpace) -> (ReturnValue_Location, ReturnValue_Rotation:Rotator, ReturnValue_Scale).ReturnValue_Rotation
    B:Rotator="5.4387,48.3711,-0.9415"
   ) -> (ReturnValue_Roll:float, ReturnValue_Pitch:float, ReturnValue_Yaw:float)
   n3 = float + float(
    A="Delta (Rotator)"(
     A=GetSocketTransform(self=@n1, InSocketName=wrist_fr_l, TransformSpace=RTS_ParentBoneSpace).ReturnValue_Rotation
     B="1.4136,-13.8593,179.4784"
    )
    B=@n2
   )
   > Set Extra_Elbow_L_Rot(Extra_Elbow_L_Rot:Rotator=MakeRotator(Roll=float * float(A=@n3, B=0.3) {is=title K2Node_PromotableOperator})) -> (Output_Get:Rotator)
   > Set Extra_Wrist_L_Rot(Extra_Wrist_L_Rot:Rotator=MakeRotator(Roll=float * float(A=@n2, B=0.5))) -> (Output_Get:Rotator)
   n4 = GetComponentByClass(self=GetOwningActor(self=Self-Reference()), ComponentClass=&SkeletalMeshComponent)
   n5 = "Delta (Rotator)"(A=GetSocketTransform(self=@n4, InSocketName=ball_fr_r, TransformSpace=RTS_ParentBoneSpace).ReturnValue_Rotation, B="5.4387,48.3711,-0.9415")
   n6 = float + float(
    A="Delta (Rotator)"(
     A=GetSocketTransform(self=@n4, InSocketName=wrist_fr_r, TransformSpace=RTS_ParentBoneSpace).ReturnValue_Rotation
     B="1.4136,-13.8593,179.4784"
    )
    B=@n5
   )
   > Set Extra_Elbow_R_Rot(Extra_Elbow_R_Rot:Rotator=MakeRotator(Roll=float * float(A=@n6, B=0.3))) -> (Output_Get:Rotator)
   > Set Extra_Wrist_R_Rot(Extra_Wrist_R_Rot:Rotator=MakeRotator(Roll=float * float(A=@n5, B=0.5))) -> (Output_Get:Rotator)
  ]}
]
variables [
 Chest_Alpha {type=number}
 Extra_Elbow_L_Rot {type=Rotator}
 Extra_Elbow_L_Rot_0 {type=Rotator}
 Extra_Elbow_R_Rot {type=Rotator}
 Extra_Elbow_R_Rot_0 {type=Rotator}
 Extra_Wrist_L_Rot {type=Rotator}
 Extra_Wrist_L_Rot_0 {type=Rotator}
 Extra_Wrist_R_Rot {type=Rotator}
 Extra_Wrist_R_Rot_0 {type=Rotator}
 PoseAsset {type=PoseAsset, default=&/GalaxySwish_Quadruped_PoseAsset}
]
